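# Post-animation style Control Rig for the male medium FeralTrash body, correcting body deformation and costume bones as limbs move.
# It runs the shared FN_DeformRig_FNC, then offsets thigh pouch and fishbone by the deformRig_r_thigh_up_alpha attribute and parent-constrains fork, spoon and back loop.
# Spherical pose readers on upper arms and thighs weight offsets on collars, vest, loops, bottle and utensils.
# Belt spline-IK and pose-blend nodes are present but unconnected to execution.
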ControlRigBlueprint /Game/Characters/Player/Male/Medium/Bodies/M_MED_FeralTrash/Meshes/M_MED_FeralTrash_CtrlRig
graphs [
 RigVMFunctionLibrary {links [], nodes []}
 RigVMFunction_Sequence_1_ContainedGraph {nodes [
   n1 = RigVMFunctionEntryNode()
   RigVMFunction_Sequence = &RigVMFunction_Sequence(ExecuteContext=@n1.ExecuteContext)
   n2 = &RigVMFunction_Sequence(ExecuteContext=@RigVMFunction_Sequence.B)
   n3 = &RigVMFunction_Sequence(ExecuteContext=@n2.B)
   n4 = &RigVMFunction_Sequence(ExecuteContext=@n3.B)
   n5 = &RigVMFunction_Sequence(ExecuteContext=@n4.B)
   n6 = &RigVMFunction_Sequence(ExecuteContext=@n5.B)
   RigVMFunctionReturnNode(A=@RigVMFunction_Sequence.A, B=@n2.A, C=@n3.A, D=@n4.A, E=@n5.A, F=@n6.A, G=@n6.B)
  ]}
 RigVMFunction_Sequence_1_ContainedGraph {nodes [
   n1 = RigVMFunctionEntryNode()
   RigVMFunction_Sequence = &RigVMFunction_Sequence(ExecuteContext=@n1.ExecuteContext)
   n2 = &RigVMFunction_Sequence(ExecuteContext=@RigVMFunction_Sequence.B)
   RigVMFunctionReturnNode(A=@RigVMFunction_Sequence.A, B=@n2.A, C=@n2.B)
  ]}
 RigVMModel {nodes [
   RigUnit_BeginExecution = &RigUnit_BeginExecution()
   RigVMFunction_Sequence = RigVMAggregateNode(ExecuteContext=@RigUnit_BeginExecution.ExecutePin)
   FN_DeformRig_FNC = RigVMFunctionReferenceNode(
    DeformRig_ Curve="(deformRig_l_upperarm_up_alpha_72_#1=False,deformRig_r_upperarm_up_alpha_73_#2=False,deformRig_l_upperarm_fr_alpha_74_#3=False,deformRig_r_upperarm_fr_alpha_75_CB51B8 ...(truncated)"
    DeformRig_Bool="(upperarm_1_#4=True,elbow_3_#5=True,wrist_7_#6=True,groin_8_#7=True,glute_17_#8=True,knee_9_40C408324CA261 ...(truncated)"
    DeformRig_Debug=(upperarm_debug_18_#4=False,wrist_debug_20_#6=False,lat_debug_26_#9=False,groinGlute_debug_27_#7=False)
    DeformRig_Value="(upperarm_up_18_#10=(Rotation=(X=0,Y=0,Z=0,W=1),Translation=(X=0,Y=0,Z=8),Scale3D=(X=0.7,Y=1,Z=1)),upperarm_fwd_fr_19_#11=(Rotation=(X=0,Y=0,Z=0,W=1),Translation=(X=0,Y=10,Z=0),Scale3D=(X=0.5,Y=1,Z= ...(truncated)"
    ExecuteContext=@RigVMFunction_Sequence.A
   )
   DISPATCH_RigDispatch_GetAnimAttribute = RigVMDispatchNode(BoneName=root, Default=0, Name=deformRig_r_thigh_up_alpha) {template="DISPATCH_RigDispatch_GetAnimAttribute(in Name,in BoneName,in Default,out Value,out Found)"}
   &RigUnit_ModifyTransforms(
    ItemToModify=((Item=(Type=Bone,Name="dyn_pouchThigh_r_MSTR"),Transform=(Rotation=(X=-0.1961,Y=0.1544,Z=-0.0983,W=0.9634),Translation=(X=-4,Y=0,Z=-1),Scale3D=(X=1,Y=1,Z=1))))
    WeightMinimum=0
    Weight=@DISPATCH_RigDispatch_GetAnimAttribute.Value
    ExecutePin=@RigVMFunction_Sequence.B
   )
   > &RigUnit_ModifyTransforms(
    ItemToModify=((Item=(Type=Bone,Name="dyn_pouchThigh_r_01"),Transform=(Rotation=(X=0,Y=0.0872,Z=0,W=0.9962),Translation=(X=0,Y=0,Z=-3),Scale3D=(X=1,Y=1,Z=1))))
    WeightMinimum=0
    Weight=@DISPATCH_RigDispatch_GetAnimAttribute.Value
   )
   > &RigUnit_ModifyTransforms(
    ItemToModify=((Item=(Type=Bone,Name="dyn_fishBone_r_MSTR"),Transform=(Rotation=(X=0,Y=0,Z=0.2588,W=0.9659),Translation=(X=-10,Y=0,Z=-2),Scale3D=(X=1,Y=1,Z=1))))
    WeightMinimum=0
    Weight=@DISPATCH_RigDispatch_GetAnimAttribute.Value
   )
   > &RigUnit_ParentConstraint(
    AdvancedSettings=(InterpolationType=Average,RotationOrderForFilter=XZY)
    Child=(Type=Bone,Name="dyn_fork_r_MSTR")
    Parents="((Item=(Type=Bone,Name=\"thigh_twist_01_r\"),Weight=0.554),(Item=(Type=Bone,Name=\"deform_groin_r\"),Weight=0.2044),(Item=(Type=Bone,Name=\"thigh_r\"),Weight=0.0569),(Item=(Type=Bone,Name=\"deform_groin_l\"),Weight=0.0187),(Item=(Type=Bone,Name=\"pelvis\"),Weight=0. ...(truncated)"
    Weight=1
   )
   > &RigUnit_ParentConstraint(
    AdvancedSettings=(InterpolationType=Average,RotationOrderForFilter=XZY)
    Child=(Type=Bone,Name="dyn_spoon_r_MSTR")
    Parents="((Item=(Type=Bone,Name=\"thigh_twist_01_r\"),Weight=0.477),(Item=(Type=Bone,Name=\"deform_groin_r\"),Weight=0.2152),(Item=(Type=Bone,Name=\"thigh_r\"),Weight=0.0422),(Item=(Type=Bone,Name=\"deform_groin_l\"),Weight=0.0378),(Item=(Type=Bone,Name=\"pelvis\"),Weight=0. ...(truncated)"
    Weight=1
   )
   > &RigUnit_ParentConstraint(
    AdvancedSettings=(InterpolationType=Average,RotationOrderForFilter=XZY)
    Child=(Type=Bone,Name="dyn_loopBk_l_MSTR")
    Parents="((Item=(Type=Bone,Name=\"deform_glute_l\"),Weight=0.3329),(Item=(Type=Bone,Name=\"pelvis\"),Weight=0.263),(Item=(Type=Bone,Name=\"thigh_twist_01_l\"),Weight=0.2566),(Item=(Type=Bone,Name=\"spine_02\"),Weight=0.0698),(Item=(Type=Bone,Name=\"spine_01\"),Weight=0.0485) ...(truncated)"
    Weight=1
   )
   n1 = &RigUnit_SphericalPoseReader(
    ActiveRegionScaleFactors=(PositiveWidth=3.5,NegativeWidth=0.5,PositiveHeight=3,NegativeHeight=3)
    ActiveRegionSize=0.1
    Debug=(bDrawDebug=true,DebugScale=25,DebugSegments=20,DebugThickness=0.25)
    DriverAxis=(X=-1,Y=0,Z=0)
    DriverItem=(Type=Bone,Name="thigh_l")
    FalloffRegionScaleFactors=(PositiveWidth=1,NegativeWidth=1,PositiveHeight=1,NegativeHeight=1)
    FalloffSize=0.5
    OptionalParentItem=(Type=Bone)
    RotationOffset=(X=110,Y=0,Z=0)
    ExecutePin=@RigVMFunction_Sequence.G
   )
   > &RigUnit_ModifyTransforms(
    ItemToModify=((Item=(Type=Bone,Name="dyn_vest_l_MSTR"),Transform=(Rotation=(X=0.262,Y=0.2198,Z=-0.7198,W=0.604),Translation=(X=-9,Y=0,Z=0),Scale3D=(X=1,Y=1,Z=1))))
    WeightMinimum=0.1
    Weight=@n1.OutputParam
   )
   > &RigUnit_ModifyTransforms(
    ItemToModify=((Item=(Type=Bone,Name="dyn_vest_l_01"),Transform=(Rotation=(X=0,Y=0,Z=0,W=1),Translation=(X=0,Y=0,Z=0),Scale3D=(X=1,Y=1,Z=1))))
    WeightMinimum=0
    Weight=@n1.OutputParam
   )
   > &RigUnit_ModifyTransforms(
    ItemToModify=((Item=(Type=Bone,Name="dyn_vest_l_02"),Transform=(Rotation=(X=0,Y=0,Z=0,W=1),Translation=(X=0,Y=0,Z=0),Scale3D=(X=1,Y=1,Z=1))))
    WeightMinimum=0
    Weight=@n1.OutputParam
   )
   > &RigUnit_ModifyTransforms(
    ItemToModify=((Item=(Type=Bone,Name="dyn_vestSide_l_MSTR"),Transform=(Rotation=(X=0.1565,Y=0.1472,Z=-0.1919,W=0.9576),Translation=(X=0.0983,Y=0,Z=0),Scale3D=(X=1,Y=1,Z=1))))
    WeightMinimum=0
    Weight=@n1.OutputParam
   )
   > &RigUnit_ModifyTransforms(
    ItemToModify=((Item=(Type=Bone,Name="dyn_loopFr_l_MSTR"),Transform=(Rotation=(X=0.119,Y=0.3137,Z=-0.2205,W=0.9159),Translation=(X=0,Y=-2,Z=0),Scale3D=(X=0.8,Y=1,Z=1))))
    WeightMinimum=0
    Weight=@n1.OutputParam
   )
   > &RigUnit_ModifyTransforms(
    ItemToModify=((Item=(Type=Bone,Name="dyn_loopFr_l_01bis"),Transform=(Rotation=(X=0,Y=0.1736,Z=0,W=0.9848),Translation=(X=0,Y=0,Z=0),Scale3D=(X=1,Y=1,Z=1))))
    WeightMinimum=0
    Weight=@n1.OutputParam
   )
   > &RigUnit_ModifyTransforms(
    ItemToModify=((Item=(Type=Bone,Name="dyn_loopFr_l_03"),Transform=(Rotation=(X=0,Y=0,Z=0.4226,W=0.9063),Translation=(X=0,Y=0,Z=0),Scale3D=(X=1,Y=1,Z=1))))
    WeightMinimum=0
    Weight=@n1.OutputParam
   )
   > &RigUnit_ModifyTransforms(
    ItemToModify=((Item=(Type=Bone,Name="dyn_loopBk_l_01"),Transform=(Rotation=(X=0.0449,Y=-0.2549,Z=0.1677,W=0.9513),Translation=(X=0,Y=0,Z=0),Scale3D=(X=1,Y=1,Z=1))))
    WeightMinimum=0
    Weight=@n1.OutputParam
   )
   > &RigUnit_ModifyTransforms(
    ItemToModify=((Item=(Type=Bone,Name="dyn_loopBk_l_02"),Transform=(Rotation=(X=0,Y=-0.2588,Z=0,W=0.9659),Translation=(X=0,Y=2,Z=0),Scale3D=(X=1,Y=1,Z=1))))
    WeightMinimum=0
    Weight=@n1.OutputParam
   )
   > n2 = &RigUnit_SphericalPoseReader(
    ActiveRegionScaleFactors=(PositiveWidth=4.5,NegativeWidth=0.5,PositiveHeight=3,NegativeHeight=3)
    ActiveRegionSize=0.1
    Debug=(bDrawDebug=false,DebugScale=25,DebugSegments=20,DebugThickness=0.25)
    DriverAxis=(X=1,Y=0,Z=0)
    DriverItem=(Type=Bone,Name="thigh_r")
    FalloffRegionScaleFactors=(PositiveWidth=1,NegativeWidth=1,PositiveHeight=1,NegativeHeight=1)
    FalloffSize=0.55
    FlipHeightScaling=false
    FlipWidthScaling=true
    OptionalParentItem=(Type=Bone)
    RotationOffset=(X=-90,Y=-20,Z=0)
   )
   > &RigUnit_ModifyTransforms(
    ItemToModify=((Item=(Type=Bone,Name="dyn_vest_r_MSTR"),Transform=(Rotation=(X=-0.2589,Y=-0.0018,Z=-0.5515,W=0.793),Translation=(X=-5.5026,Y=0.6306,Z=0.4932),Scale3D=(X=1,Y=1,Z=1))))
    WeightMinimum=0
    Weight=@n2.OutputParam
   )
   > &RigUnit_ModifyTransforms(
    ItemToModify=((Item=(Type=Bone,Name="dyn_vest_r_02"),Transform=(Rotation=(X=0,Y=0,Z=0,W=1),Translation=(X=0,Y=0,Z=0),Scale3D=(X=1,Y=1,Z=1))))
    WeightMinimum=0
    Weight=@n2.OutputParam
   )
   > &RigUnit_ModifyTransforms(
    ItemToModify=((Item=(Type=Bone,Name="dyn_bottle_l_MSTR"),Transform=(Rotation=(X=-0.0394,Y=0.276,Z=-0.1749,W=0.9443),Translation=(X=0,Y=-1.083,Z=-2.5753),Scale3D=(X=1,Y=1,Z=1))))
    WeightMinimum=0
    Weight=@n2.OutputParam
   )
   > &RigUnit_ModifyTransforms(
    ItemToModify=((Item=(Type=Bone,Name="dyn_vestSide_r_MSTR"),Transform=(Rotation=(X=0,Y=0,Z=-0.1285,W=0.9917),Translation=(X=0,Y=0,Z=0),Scale3D=(X=1,Y=1,Z=1))))
    WeightMinimum=0
    Weight=@n2.OutputParam
   )
   > &RigUnit_ModifyTransforms(
    ItemToModify=((Item=(Type=Bone,Name="dyn_fork_r_MSTR"),Transform=(Rotation=(X=0,Y=0,Z=0.2164,W=0.9763),Translation=(X=0,Y=1,Z=0),Scale3D=(X=1,Y=1,Z=1))))
    WeightMinimum=0
    Weight=@n2.OutputParam
   )
   > &RigUnit_ModifyTransforms(
    ItemToModify=((Item=(Type=Bone,Name="dyn_spoon_r_MSTR"),Transform=(Rotation=(X=0,Y=0,Z=0.2588,W=0.9659),Translation=(X=0,Y=1,Z=0),Scale3D=(X=1,Y=1,Z=1))))
    WeightMinimum=0
    Weight=@n2.OutputParam
   )
   n3 = &RigUnit_SphericalPoseReader(
    ActiveRegionScaleFactors=(PositiveWidth=2.5,NegativeWidth=3,PositiveHeight=3,NegativeHeight=3)
    ActiveRegionSize=0.1
    Debug=(bDrawDebug=false,DebugScale=25,DebugSegments=20,DebugThickness=0.25)
    DriverAxis=(X=1,Y=0,Z=0)
    DriverItem=(Type=Bone,Name="upperarm_l")
    FalloffRegionScaleFactors=(PositiveWidth=1,NegativeWidth=1,PositiveHeight=1,NegativeHeight=1)
    FalloffSize=0.5
    OptionalParentItem=(Type=Bone)
    RotationOffset=(X=0,Y=0,Z=0)
    ExecutePin=@RigVMFunction_Sequence.F
   )
   > &RigUnit_ModifyTransforms(
    ItemToModify=((Item=(Type=Bone,Name="dyn_collarA_l_MSTR"),Transform=(Rotation=(X=0,Y=0.4226,Z=0,W=0.9063),Translation=(X=0,Y=0,Z=-10),Scale3D=(X=1,Y=1,Z=1))))
    WeightMinimum=0
    Weight=@n3.OutputParam
   )
   > n4 = &RigUnit_SphericalPoseReader(
    ActiveRegionScaleFactors=(PositiveWidth=2.5,NegativeWidth=3,PositiveHeight=3,NegativeHeight=3)
    ActiveRegionSize=0.05
    Debug=(bDrawDebug=false,DebugScale=25,DebugSegments=20,DebugThickness=0.25)
    DriverAxis=(X=-1,Y=0,Z=0)
    DriverItem=(Type=Bone,Name="upperarm_r")
    FalloffRegionScaleFactors=(PositiveWidth=1,NegativeWidth=1,PositiveHeight=1,NegativeHeight=1)
    FalloffSize=0.5
    FlipHeightScaling=true
    OptionalParentItem=(Type=Bone)
    RotationOffset=(X=160,Y=-60,Z=160)
   )
   > &RigUnit_ModifyTransforms(
    ItemToModify=((Item=(Type=Bone,Name="dyn_collarA_r_MSTR"),Transform=(Rotation=(X=0.173,Y=-0.0858,Z=0.0151,W=0.9811),Translation=(X=0,Y=0,Z=2),Scale3D=(X=1,Y=1,Z=1))))
    WeightMinimum=0
    Weight=@n4.OutputParam
   )
   SequenceAggregate_2 = &RigVMFunction_Sequence(ExecuteContext=())
   n5 = &RigVMFunction_AlphaInterp(
    InterpSpeedDecreasing=20
    InterpSpeedIncreasing=10
    Scale=1
    bClampResult=False
    Value=RigVMVariableNode() {variable=thigh_r_SPR_outputParam}.Value
   )
   n6 = RigVMVariableNode() {variable=beltSpline_ctrls}
   /Script/ControlRig.RigUnit_HierarchySetPoseItemArray(
    ElementType=All
    ExecutePin=@SequenceAggregate_2.A
    Weight=&RigVMFunction_MathDoubleSub(A=1, B=@n5.Result).Result
    ItemsToSet=@n6.Value
    Pose=RigVMVariableNode() {variable=spline_ctrls_bindPose}.Value
   ) {is=struct RigVMUnitNode}
   > &RigUnit_HierarchySetPoseItemArray(ElementType=All, Weight=@n5.Result, ItemsToSet=@n6.Value, Pose=RigVMVariableNode() {variable=spline_ctrls_thigh_r_up_pose}.Value)
   &RigUnit_HierarchySetPoseItemArray(
    ElementType=All
    ItemsToSet=((Type=Control,Name="belt_start_ctrl"))
    Pose=(HierarchyTopologyVersion=-1,PoseHash=-1)
    Weight=&RigVMFunction_MathDoubleSub(A=1, B=0).Result
    ExecutePin=@SequenceAggregate_2.B
   )
   > &RigUnit_HierarchySetPoseItemArray(ElementType=All, ItemsToSet=((Type=Control,Name="belt_start_ctrl")), Pose=(HierarchyTopologyVersion=-1,PoseHash=-1), Weight=1)
   n7 = &RigUnit_SphericalPoseReader(
    ActiveRegionScaleFactors=(PositiveWidth=3.5,NegativeWidth=0.5,PositiveHeight=3,NegativeHeight=3)
    ActiveRegionSize=0.1
    Debug=(bDrawDebug=false,bDraw2D=false,DebugScale=25,DebugSegments=20,DebugThickness=0.25)
    DriverAxis=(X=1,Y=0,Z=0)
    DriverItem=(Type=Bone,Name="thigh_r")
    ExecutePin=()
    FalloffRegionScaleFactors=(PositiveWidth=1,NegativeWidth=1,PositiveHeight=1,NegativeHeight=1)
    FalloffSize=0.45
    FlipHeightScaling=false
    FlipWidthScaling=true
    OptionalParentItem=(Type=Bone)
    RotationOffset=(X=-60,Y=0,Z=0)
   )
   > RigVMVariableNode(Value=@n7.OutputParam) {variable=thigh_r_SPR_outputParam}
   > /Script/ControlRig.RigUnit_SetTranslation(
    Item=(Type=Control,Name="dyn_beltSplineIK_10_ctrl")
    Space=LocalSpace
    Value=(X=-2.3541,Y=0.0036,Z=24.1622)
    bInitial=false
    bPropagateToChildren=false
    Weight=RigVMVariableNode() {variable=thigh_r_SPR_outputParam}.Value
   ) {is=struct RigVMUnitNode, template="Set Transform::Execute(in Item,in Space,in bInitial,in Value,in Weight,in bPropagateToChildren)"}
   n8 = &RigUnit_ItemArray(Items="((Type=Control,Name=\"dyn_beltSplineIK_01_ctrl\"),(Type=Control,Name=\"dyn_beltSplineIK_04_ctrl\"),(Type=Control,Name=\"dyn_beltSplineIK_07_ctrl\"),(Type=Control,Name=\"dyn_beltSplineIK_10_ctrl\"),(Type=Control,Name=\"dyn_beltSplineIK_13_ctrl\"),(Type=Control,Name=\" ...(truncated)")
   RigVMVariableNode(ExecuteContext=(), Value=@n8.Items) {variable=beltSpline_ctrls}
   > RigVMVariableNode(Value=/Script/ControlRig.RigUnit_HierarchyGetPoseItemArray(ElementType=All, Initial=false, ItemsToGet=@n8.Items) {is=struct RigVMUnitNode}.Pose) {variable=beltSpline_ctrls_pose}
   RigUnit_ItemArray = &RigUnit_ItemArray(Items="((Type=Bone,Name=\"dyn_beltSplineIK_01\"),(Type=Bone,Name=\"dyn_beltSplineIK_02\"),(Type=Bone,Name=\"dyn_beltSplineIK_03\"),(Type=Bone,Name=\"dyn_beltSplineIK_04\"),(Type=Bone,Name=\"dyn_beltSplineIK_05\"),(Type=Bone,Name=\"dyn_beltSplineIK_06\"),(Type=Bone,Name=\"dyn_ ...(truncated)")
   > RigVMVariableNode(Value=@RigUnit_ItemArray.Items) {variable=beltSpline_joints}
   Cast = &RigVMFunction_MathBoolToFloat(Value=False)
   Multiply = &RigVMFunction_MathDoubleMul(B=1, A=@Cast.Result)
   Subtract = &RigVMFunction_MathDoubleSub(A=1, B=@Cast.Result)
   Maximum = &RigVMFunction_MathDoubleMax(A=@Multiply.Result, B=&RigVMFunction_MathDoubleMul(B=1, A=@Subtract.Result).Result)
   &RigUnit_ParentConstraint(
    AdvancedSettings=(InterpolationType=Average,RotationOrderForFilter=XZY)
    Child=(Type=Bone,Name="dyn_pouchThigh_r_MSTR")
    ExecutePin=()
    Parents=((Item=(Type=Bone,Name="pelvis"),Weight=0.5),(Item=(Type=Bone,Name="thigh_r"),Weight=1))
    Weight=1
   )
   SplineIK_1 = RigVMFunctionReferenceNode(
    ExecuteContext=()
    Primary Axis=(X=1,Y=0,Z=0)
    Samples Per Segment=30
    Secondary Spline Direction=(X=0,Y=0,Z=0)
    Spline Mode=BSpline
    Stretch=true
    Up Axis=(X=0,Y=0,Z=0)
    debug=true
    Controls=RigVMVariableNode() {variable=beltSpline_ctrls}.Value
    Bones=RigVMVariableNode() {variable=beltSpline_joints}.Value
   )
  ]}
 RigVMModel Construction {nodes [
   RigVMFunction_Sequence = RigVMAggregateNode(ExecuteContext=())
   RigVMVariableNode(
    Value=&RigUnit_HierarchyGetPoseItemArray(ElementType=All, ItemsToGet=RigVMVariableNode() {variable=beltSpline_ctrls}.Value).Pose
    ExecuteContext=@RigVMFunction_Sequence.A
   ) {variable=spline_ctrls_bindPose}
   &RigUnit_ModifyTransforms(
    ItemToModify=((Item=(Type=Control,Name="dyn_beltSplineIK_07_ctrl"),Transform=(Rotation=(X=0,Y=0,Z=0,W=1),Translation=(X=0,Y=-5.3786,Z=0.9659),Scale3D=(X=1,Y=1,Z=1))))
    Weight=1
    WeightMinimum=0
    ExecutePin=@RigVMFunction_Sequence.B
   )
   > &RigUnit_ModifyTransforms(
    ItemToModify=((Item=(Type=Control,Name="dyn_beltSplineIK_10_ctrl"),Transform=(Rotation=(X=0,Y=0,Z=0,W=1),Translation=(X=2.6811,Y=-19.756,Z=1.5099),Scale3D=(X=1,Y=1,Z=1))))
    Weight=1
    WeightMinimum=0
   )
   > &RigUnit_ModifyTransforms(
    ItemToModify=((Item=(Type=Control,Name="dyn_beltSplineIK_13_ctrl"),Transform=(Rotation=(X=0,Y=0,Z=0,W=1),Translation=(X=0,Y=-15.785,Z=5.274),Scale3D=(X=1,Y=1,Z=1))))
    Weight=1
    WeightMinimum=0
   )
   > &RigUnit_ModifyTransforms(
    ItemToModify=((Item=(Type=Control,Name="dyn_beltSplineIK_16_ctrl"),Transform=(Rotation=(X=0,Y=0,Z=0,W=1),Translation=(X=0,Y=-14.2417,Z=2.6912),Scale3D=(X=1,Y=1,Z=1))))
    Weight=1
    WeightMinimum=0
   )
   > RigVMVariableNode(Value=&RigUnit_HierarchyGetPoseItemArray(ElementType=All, ItemsToGet=RigVMVariableNode() {variable=beltSpline_ctrls}.Value).Pose) {variable=spline_ctrls_thigh_r_up_pose}
   &RigUnit_ModifyTransforms(
    ItemToModify=((Item=(Type=Control,Name="dyn_beltSplineIK_18_ctrl"),Transform=(Rotation=(X=0,Y=-0.1736,Z=0,W=0.9848),Translation=(X=0,Y=0,Z=0),Scale3D=(X=1,Y=1,Z=1))))
    Weight=1
    WeightMinimum=0
    ExecutePin=@RigVMFunction_Sequence.C
   )
   > &RigUnit_ModifyTransforms(
    ItemToModify=((Item=(Type=Control,Name="dyn_beltSplineIK_16_ctrl"),Transform=(Rotation=(X=-0.1736,Y=0,Z=0,W=0.9848),Translation=(X=0,Y=0,Z=0),Scale3D=(X=1,Y=1,Z=1))))
    Weight=1
    WeightMinimum=0
   )
   > &RigUnit_ModifyTransforms(
    ItemToModify=((Item=(Type=Control,Name="dyn_beltSplineIK_13_ctrl"),Transform=(Rotation=(X=0,Y=-0.0872,Z=0,W=0.9962),Translation=(X=0,Y=0,Z=0),Scale3D=(X=1,Y=1,Z=1))))
    Weight=1
    WeightMinimum=0
   )
   > RigVMVariableNode(Value=&RigUnit_HierarchyGetPoseItemArray(ElementType=All, ItemsToGet=RigVMVariableNode() {variable=beltSpline_ctrls}.Value).Pose) {variable=spline_ctrls_thigh_l_up_pose}
   &RigUnit_PrepareForExecution()
  ]}
]
hierarchy [
 root {loc=<0 0 0>, rot=<0 0 0 1>}
 pelvis {parent=root, loc=<0 -0.7071 93.1191>, rot=<0.0127 0.707 -0.0127 -0.707>}
 spine_01 {parent=pelvis, loc=<6.7494 -0.6388 0>, rot=<0 0 0.034 -0.9994>}
 spine_02 {parent=spine_01, loc=<7.2518 -0.035 0>, rot=<0 0 0.0238 -0.9997>}
 spine_03 {parent=spine_02, loc=<6.0104 0 0>, rot=<0 0 0.0149 -0.9999>}
 spine_04 {parent=spine_03, loc=<8.1024 0 0>, rot=<0 0 -0.0132 -0.9999>}
 spine_05 {parent=spine_04, loc=<10.0146 0 0>, rot=<0 0 -0.0015 -1>}
 attach_cape {parent=spine_05, loc=<9.017 -11.714 0>, rot=<0.0616 -0.7044 -0.0616 -0.7044>}
 neck_01 {parent=spine_05, loc=<11.3138 0.1529 0>, rot=<0 0 -0.2003 -0.9797>}
 neck_02 {parent=neck_01, loc=<4.5627 0 0>, rot=<0 0 0.0504 -0.9987>}
 head {parent=neck_02, loc=<5.3914 0 0>, rot=<0 0 0.115 -0.9934>}
 attach_backpack {parent=spine_05, loc=<-0.123 -13.473 0>, rot=<-0.0093 -0.707 0.0093 -0.707>}
 clavicle_l {parent=spine_05, loc=<4.3575 -0.1466 -2.3223>, rot=<0.0465 -0.7289 -0.009 -0.6829>}
 upperarm_l {parent=clavicle_l, loc=<15.5111 0 0>, rot=<0.0226 -0.3234 -0.0088 -0.9459>}
 lowerarm_l {parent=upperarm_l, loc=<26.6514 0 0>, rot=<0.0391 -0.0099 -0.2453 -0.9686>}
 hand_l {parent=lowerarm_l, loc=<22.3491 0 0>, rot=<-0.7081 -0.0507 0.0975 -0.6975>}
 thumb_01_l {parent=hand_l, loc=<1.3118 -0.3191 -2.459>, rot=<0.6546 -0.4042 0.0664 -0.6354>}
 thumb_02_l {parent=thumb_01_l, loc=<5.9336 -0.0448 0.0401>, rot=<-0.0037 0.0213 0.298 -0.9543>}
 thumb_03_l {parent=thumb_02_l, loc=<2.2223 0.3682 -0.0907>, rot=<0.0118 -0.013 0.0066 -0.9998>}
 index_metacarpal_l {parent=hand_l, loc=<1.6535 0.0183 -2.96>, rot=<0.0158 -0.0124 -0.0173 -0.9996>}
 index_01_l {parent=index_metacarpal_l, loc=<8.3008 0.0357 -0.0524>, rot=<-0.0044 0.0485 0.0924 -0.9945>}
 index_02_l {parent=index_01_l, loc=<3.3587 -0.3004 0.0452>, rot=<-0.002 0.0027 0.3225 -0.9466>}
 index_03_l {parent=index_02_l, loc=<2.5578 -0.123 -0.0252>, rot=<-0.002 0.0067 0.2558 -0.9667>}
 middle_metacarpal_l {parent=hand_l, loc=<2.1856 -0.1725 -0.5984>, rot=<0.006 0.0168 -0.0329 -0.9993>}
 middle_01_l {parent=middle_metacarpal_l, loc=<7.3275 -0.2447 0.0129>, rot=<-0.0208 0.0042 0.0858 -0.9961>}
 middle_02_l {parent=middle_01_l, loc=<4.0123 0.1401 0.1175>, rot=<-0.0218 0.0258 0.3512 -0.9357>}
 middle_03_l {parent=middle_02_l, loc=<2.6077 0.0534 0.1373>, rot=<0.0082 0.0188 0.1708 -0.9851>}
 ring_metacarpal_l {parent=hand_l, loc=<2.2074 -0.3625 1.3638>, rot=<-0.0284 0.0722 -0.0458 -0.9959>}
 ring_01_l {parent=ring_metacarpal_l, loc=<7.0295 -0.6197 -0.1446>, rot=<-0.0489 -0.0252 0.1892 -0.9804>}
 ring_02_l {parent=ring_01_l, loc=<3.6174 0.4647 -0.1027>, rot=<0 0 0.2806 -0.9598>}
 ring_03_l {parent=ring_02_l, loc=<2.4152 0.114 0.0199>, rot=<0 0 0.1189 -0.9929>}
 pinky_metacarpal_l {parent=hand_l, loc=<1.7616 -0.9753 3.3814>, rot=<-0.0397 0.0814 -0.069 -0.9935>}
 pinky_01_l {parent=pinky_metacarpal_l, loc=<7.1709 -0.4831 -0.1094>, rot=<-0.0654 -0.0035 0.2242 -0.9723>}
 pinky_02_l {parent=pinky_01_l, loc=<2.8274 0.0893 -0.1043>, rot=<0.0037 0.0011 0.2854 -0.9584>}
 pinky_03_l {parent=pinky_02_l, loc=<2.2046 0.0307 -0.0078>, rot=<-0.0002 0.0098 0.1445 -0.9895>}
 deform_wrist_dn_base_l {parent=hand_l, loc=<0 -2.2342 0>, rot=<0 0 0 -1>}
 deform_wrist_dn_l {parent=deform_wrist_dn_base_l, loc=<0 0 0>, rot=<0 0 0 1>}
 deform_wrist_up_base_l {parent=hand_l, loc=<0 2.2682 0>, rot=<0 0 0 -1>}
 deform_wrist_up_l {parent=deform_wrist_up_base_l, loc=<0 0 0>, rot=<0 0 0 1>}
 weapon_l {parent=hand_l, loc=<-0.1974 -3.4111 0.3811>, rot=<0 0 0 1>}
 hand_attach_l {parent=hand_l, loc=<8.5 -3 0>, rot=<0.7071 0.7071 0 0>}
 lowerarm_twist_02_l {parent=lowerarm_l, loc=<9.6199 0 0>, rot=<0 0 0 -1>}
 lowerarm_twist_01_l {parent=lowerarm_l, loc=<17.8549 0 0>, rot=<0 0 0 -1>}
 deform_elbow_in_base_l {parent=lowerarm_l, loc=<0.2112 5.259 0.8274>, rot=<0 0 0 -1>}
 deform_elbow_in_l {parent=deform_elbow_in_base_l, loc=<0 0 0>, rot=<0 0 0 1>}
 deform_elbow_out_base_l {parent=lowerarm_l, loc=<0 -4 0>, rot=<0 0 0 -1>}
 deform_elbow_out_l {parent=deform_elbow_out_base_l, loc=<0 0 0>, rot=<0 0 0 1>}
 upperarm_twist_01_l {parent=upperarm_l, loc=<0.0041 0 0>, rot=<0 0 0 1>}
 deform_upperarm_bk_base_l {parent=upperarm_twist_01_l, loc=<1.0813 -7.2193 -1.9453>, rot=<0 0 0 1>}
 deform_upperarm_bk_l {parent=deform_upperarm_bk_base_l, loc=<0 0 0>, rot=<0 0 0 1>}
 deform_upperarm_fr_base_l {parent=upperarm_twist_01_l, loc=<-0.3691 7.0506 -2.765>, rot=<0 0 0 1>}
 deform_upperarm_fr_l {parent=deform_upperarm_fr_base_l, loc=<0 0 0>, rot=<0 0 0 1>}
 dyn_shoulderBuckle_l_MSTR {parent=deform_upperarm_fr_l, loc=<2.1627 4.0034 -0.7024>, rot=<0.4409 -0.3073 -0.2703 -0.7988>}
 dyn_shoulderBuckle_l_01 {parent=dyn_shoulderBuckle_l_MSTR, loc=<0 0 0>, rot=<0.0001 0.0053 -0.0005 -1>}
 deform_upperarm_up_base_l {parent=upperarm_twist_01_l, loc=<-3.2216 -1.1391 3.3311>, rot=<0 0 0 1>}
 deform_upperarm_up_l {parent=deform_upperarm_up_base_l, loc=<0 0 0>, rot=<0 0 0 1>}
 upperarm_twist_02_l {parent=upperarm_l, loc=<13.7941 0 0>, rot=<0 0 0 1>}
 clavicle_r {parent=spine_05, loc=<4.3575 -0.1466 2.3223>, rot=<0.7289 0.0465 -0.6829 0.009>}
 upperarm_r {parent=clavicle_r, loc=<-15.5111 0 0>, rot=<0.0226 -0.3234 -0.0088 -0.9459>}
 lowerarm_r {parent=upperarm_r, loc=<-26.6514 0 0>, rot=<0.0391 -0.0099 -0.2453 -0.9686>}
 hand_r {parent=lowerarm_r, loc=<-22.3491 0 0>, rot=<-0.7081 -0.0507 0.0975 -0.6975>}
 thumb_01_r {parent=hand_r, loc=<-1.3118 0.3191 2.459>, rot=<0.6546 -0.4042 0.0664 -0.6354>}
 thumb_02_r {parent=thumb_01_r, loc=<-5.9336 0.0448 -0.0401>, rot=<-0.0037 0.0213 0.298 -0.9543>}
 thumb_03_r {parent=thumb_02_r, loc=<-2.2223 -0.3682 0.0907>, rot=<0.0118 -0.013 0.0066 -0.9998>}
 index_metacarpal_r {parent=hand_r, loc=<-1.6535 -0.0183 2.96>, rot=<0.0158 -0.0124 -0.0173 -0.9996>}
 index_01_r {parent=index_metacarpal_r, loc=<-8.3008 -0.0357 0.0524>, rot=<-0.0044 0.0485 0.0924 -0.9945>}
 index_02_r {parent=index_01_r, loc=<-3.3587 0.3004 -0.0452>, rot=<-0.002 0.0027 0.3225 -0.9466>}
 index_03_r {parent=index_02_r, loc=<-2.5578 0.123 0.0252>, rot=<-0.002 0.0067 0.2558 -0.9667>}
 middle_metacarpal_r {parent=hand_r, loc=<-2.1856 0.1725 0.5984>, rot=<0.006 0.0168 -0.0329 -0.9993>}
 middle_01_r {parent=middle_metacarpal_r, loc=<-7.3275 0.2447 -0.0129>, rot=<-0.0208 0.0042 0.0858 -0.9961>}
 middle_02_r {parent=middle_01_r, loc=<-4.0123 -0.1401 -0.1175>, rot=<-0.0218 0.0258 0.3512 -0.9357>}
 middle_03_r {parent=middle_02_r, loc=<-2.6077 -0.0534 -0.1373>, rot=<0.0082 0.0188 0.1708 -0.9851>}
 ring_metacarpal_r {parent=hand_r, loc=<-2.2074 0.3625 -1.3638>, rot=<-0.0284 0.0722 -0.0458 -0.9959>}
 ring_01_r {parent=ring_metacarpal_r, loc=<-7.0295 0.6197 0.1446>, rot=<-0.0489 -0.0252 0.1892 -0.9804>}
 ring_02_r {parent=ring_01_r, loc=<-3.6174 -0.4647 0.1027>, rot=<0 0 0.2806 -0.9598>}
 ring_03_r {parent=ring_02_r, loc=<-2.4152 -0.114 -0.0199>, rot=<0 0 0.1189 -0.9929>}
 pinky_metacarpal_r {parent=hand_r, loc=<-1.7616 0.9753 -3.3814>, rot=<-0.0397 0.0814 -0.069 -0.9935>}
 pinky_01_r {parent=pinky_metacarpal_r, loc=<-7.1709 0.4831 0.1094>, rot=<-0.0654 -0.0035 0.2242 -0.9723>}
 pinky_02_r {parent=pinky_01_r, loc=<-2.8274 -0.0893 0.1043>, rot=<0.0037 0.0011 0.2854 -0.9584>}
 pinky_03_r {parent=pinky_02_r, loc=<-2.2046 -0.0307 0.0078>, rot=<-0.0002 0.0098 0.1445 -0.9895>}
 deform_wrist_dn_base_r {parent=hand_r, loc=<-0.0004 2.2345 -0.0001>, rot=<0 0 0 -1>}
 deform_wrist_dn_r {parent=deform_wrist_dn_base_r, loc=<0 0 0>, rot=<0 0 0 1>}
 deform_wrist_up_base_r {parent=hand_r, loc=<0.0003 -2.2684 0.0001>, rot=<0 0 0 -1>}
 deform_wrist_up_r {parent=deform_wrist_up_base_r, loc=<0 0 0>, rot=<0 0 0 1>}
 weapon_r {parent=hand_r, loc=<0.1974 3.4111 -0.3811>, rot=<0 0 0 1>}
 hand_attach_r {parent=hand_r, loc=<-8.5001 2.9999 0>, rot=<0 0 -0.7071 -0.7071>}
 lowerarm_twist_02_r {parent=lowerarm_r, loc=<-9.6199 0 0>, rot=<0 0 0 -1>}
 lowerarm_twist_01_r {parent=lowerarm_r, loc=<-17.8549 0 0>, rot=<0 0 0 -1>}
 deform_elbow_in_base_r {parent=lowerarm_r, loc=<-0.2111 -5.2591 -0.8275>, rot=<0 0 0 -1>}
 deform_elbow_in_r {parent=deform_elbow_in_base_r, loc=<-0.0001 0 0.0002>, rot=<0 0 0 1>}
 deform_elbow_out_base_r {parent=lowerarm_r, loc=<0 4 -0.0001>, rot=<0 0 0 -1>}
 deform_elbow_out_r {parent=deform_elbow_out_base_r, loc=<0 0 0>, rot=<0 0 0 1>}
 upperarm_twist_01_r {parent=upperarm_r, loc=<-0.0041 0 0>, rot=<0 0 0 -1>}
 deform_upperarm_bk_base_r {parent=upperarm_twist_01_r, loc=<-1.0814 7.2192 1.9455>, rot=<0 0 0 -1>}
 deform_upperarm_bk_r {parent=deform_upperarm_bk_base_r, loc=<0 0 0>, rot=<0 0 0 1>}
 deform_upperarm_fr_base_r {parent=upperarm_twist_01_r, loc=<0.3689 -7.0506 2.7652>, rot=<0 0 0 -1>}
 deform_upperarm_fr_r {parent=deform_upperarm_fr_base_r, loc=<0 0 0>, rot=<0 0 0 1>}
 deform_upperarm_up_base_r {parent=upperarm_twist_01_r, loc=<3.2214 1.1391 -3.3308>, rot=<0 0 0 -1>}
 deform_upperarm_up_r {parent=deform_upperarm_up_base_r, loc=<0 0 0>, rot=<0 0 0 -1>}
 upperarm_twist_02_r {parent=upperarm_r, loc=<-13.7941 0 0>, rot=<0 0 0 -1>}
 dyn_collarA_l_MSTR {parent=spine_05, loc=<2.2074 10.6413 -9.4655>, rot=<0.0108 -0.2315 -0.1731 -0.9572>}
 dyn_collarA_l_01 {parent=dyn_collarA_l_MSTR, loc=<0 0 0>, rot=<0 0 0 1>}
 dyn_collarB_l_MSTR {parent=spine_05, loc=<12.1574 -8.7973 -8.4276>, rot=<0.4192 -0.0656 -0.0397 -0.9047>}
 dyn_collarB_l_01 {parent=dyn_collarB_l_MSTR, loc=<0 0 0>, rot=<0 0 0 1>}
 dyn_collarA_r_MSTR {parent=spine_05, loc=<2.9696 10.7458 10.6659>, rot=<0.9585 0.149 0.2429 -0.0093>}
 dyn_collarA_r_01 {parent=dyn_collarA_r_MSTR, loc=<0 0 0>, rot=<0 0 0 1>}
 dyn_collarB_r_MSTR {parent=spine_05, loc=<12.0201 -9.153 8.0906>, rot=<0.94 0.0273 0.1131 -0.3206>}
 dyn_collarB_r_01 {parent=dyn_collarB_r_MSTR, loc=<0 0 0>, rot=<0 0 0 1>}
 dyn_collar_c_MSTR {parent=spine_05, loc=<12.1996 -11.499 0.0169>, rot=<0.706 -0.0397 0.0397 -0.706>}
 dyn_collar_c_01 {parent=dyn_collar_c_MSTR, loc=<0 0 0>, rot=<0 0 0 1>}
 deform_pec_aim_base_l {parent=spine_04, loc=<7.8732 12.6388 -2.4123>, rot=<0.0102 -0.4602 0.3979 -0.7936>}
 deform_pec_aim_l {parent=deform_pec_aim_base_l, loc=<0 0 0>, rot=<0 0 0 1>}
 deform_pec_base_l {parent=deform_pec_aim_l, loc=<10.766 0 0>, rot=<0 0 0 1>}
 deform_pec_l {parent=deform_pec_base_l, loc=<0.055 0 0>, rot=<0 0 0 1>}
 deform_pec_aim_base_r {parent=spine_04, loc=<7.8727 12.6388 2.4123>, rot=<-0.4602 -0.0102 0.7936 0.3979>}
 deform_pec_aim_r {parent=deform_pec_aim_base_r, loc=<0 0 0>, rot=<0 0 0 1>}
 deform_pec_base_r {parent=deform_pec_aim_r, loc=<-10.7662 -0.0006 -0.0007>, rot=<0 0 0 1>}
 deform_pec_r {parent=deform_pec_base_r, loc=<0 0 0>, rot=<0 0 0 1>}
 dyn_vest_l_MSTR {parent=spine_02, loc=<-1.3702 10.6051 -9.7727>, rot=<-0.0512 0.4004 0.9147 0.0183>}
 dyn_vest_l_01 {parent=dyn_vest_l_MSTR, loc=<0 0 0>, rot=<0 0 0 1>}
 dyn_vest_l_02 {parent=dyn_vest_l_01, loc=<12.0571 0 0>, rot=<0 0 0 1>}
 dyn_beltBuckle_l_MSTR {parent=dyn_vest_l_01, loc=<8.6148 -2.4838 -0.0087>, rot=<0.0027 0.6711 -0.0078 -0.7413>}
 dyn_beltBuckle_l_01 {parent=dyn_beltBuckle_l_MSTR, loc=<0 0 0>, rot=<0 0 0 1>}
 dyn_vest_r_MSTR {parent=spine_02, loc=<-2.4997 11.0238 9.1978>, rot=<0.0449 -0.3092 0.9494 0.0311>}
 dyn_vest_r_01 {parent=dyn_vest_r_MSTR, loc=<0 0 0>, rot=<0 0 0 1>}
 dyn_vest_r_02 {parent=dyn_vest_r_01, loc=<11.5048 0 0>, rot=<0 0 0 1>}
 dyn_bottle_l_MSTR {parent=spine_02, loc=<1.172 7.3239 14.1403>, rot=<-0.0037 -0.9345 -0.3558 0.0085>}
 dyn_bottle_l_01 {parent=dyn_bottle_l_MSTR, loc=<0 0 0>, rot=<0 0 0 1>}
 dyn_vestBk_l_MSTR {parent=spine_01, loc=<4.3332 -10.8949 -12.2591>, rot=<0.085 -0.9376 -0.3367 -0.0196>}
 dyn_vestBk_l_01 {parent=dyn_vestBk_l_MSTR, loc=<0 0 0>, rot=<0 0 0 1>}
 dyn_vestBk_c_MSTR {parent=spine_01, loc=<4.6071 -13.5209 -0.0019>, rot=<0.0955 -0.9954 0 0>}
 dyn_vestBk_c_01 {parent=dyn_vestBk_c_MSTR, loc=<0 0 0>, rot=<0 0 0 1>}
 dyn_vestBk_r_MSTR {parent=spine_01, loc=<4.3332 -10.8949 12.259>, rot=<0.1044 -0.9371 0.3318 0.0284>}
 dyn_vestBk_r_01 {parent=dyn_vestBk_r_MSTR, loc=<0 0 0>, rot=<0 0 0 1>}
 dyn_vestSide_l_MSTR {parent=spine_01, loc=<3.4906 -2.8172 -15.9923>, rot=<0.1718 -0.7251 -0.6621 -0.0801>}
 dyn_vestSide_l_01 {parent=dyn_vestSide_l_MSTR, loc=<0 0 0>, rot=<0 0 0 1>}
 dyn_vestSide_r_MSTR {parent=spine_01, loc=<3.5208 -2.976 16.0881>, rot=<0.1215 -0.7467 0.6501 0.0715>}
 dyn_vestSide_r_01 {parent=dyn_vestSide_r_MSTR, loc=<0 0 0>, rot=<0 0 0 1>}
 dyn_beltLip_r_MSTR {parent=spine_01, loc=<-1.0446 11.061 13.2215>, rot=<-0.4243 -0.6255 -0.4613 -0.4646>}
 dyn_beltLip_r_01 {parent=dyn_beltLip_r_MSTR, loc=<0 0 0>, rot=<0 0 0 1>}
 thigh_l {parent=pelvis, loc=<-2.5693 -0.0109 -7.7501>, rot=<0.0012 0.0572 0.0155 -0.9982>}
 calf_l {parent=thigh_l, loc=<-39.0484 0 0>, rot=<0.0006 -0.0161 -0.0391 -0.9991>}
 foot_l {parent=calf_l, loc=<-42.561 0.1077 0.4292>, rot=<0.0418 -0.0455 0.0343 -0.9975>}
 ball_l {parent=foot_l, loc=<-6.1599 13.8264 0.6806>, rot=<0.0001 -0.0001 -0.7063 -0.7079>}
 calf_twist_02_l {parent=calf_l, loc=<-18.8064 -0.0967 0.0397>, rot=<0.0237 0.0095 -0.0035 -0.9997>}
 calf_twist_01_l {parent=calf_l, loc=<-35.813 -0.0967 0.0397>, rot=<0.0237 0.0095 -0.0035 -0.9997>}
 deform_knee_in_base_l {parent=calf_l, loc=<1.298 -5.2957 -0.0496>, rot=<0 0 0 -1>}
 deform_knee_in_l {parent=deform_knee_in_base_l, loc=<0 0 0>, rot=<0 0 0 1>}
 deform_knee_out_base_l {parent=calf_l, loc=<0 4 0>, rot=<0 0 0 -1>}
 deform_knee_out_l {parent=deform_knee_out_base_l, loc=<0 0 0>, rot=<0 0 0 1>}
 thigh_twist_01_l {parent=thigh_l, loc=<-21.6303 0 0>, rot=<0.0013 0 0 -1>}
 deform_groin_base_l {parent=thigh_l, loc=<-2.1908 8.1142 -1.7442>, rot=<0 0 0 -1>}
 deform_groin_l {parent=deform_groin_base_l, loc=<0 0 0>, rot=<0 0 0 1>}
 dyn_loopFr_l_MSTR {parent=deform_groin_l, loc=<-1.559 1.4256 -0.921>, rot=<-0.0305 0.2375 0.9615 0.1347>}
 dyn_loopFr_l_01 {parent=dyn_loopFr_l_MSTR, loc=<0 0 0>, rot=<0 0 0 1>}
 dyn_loopFr_l_01bis {parent=dyn_loopFr_l_01, loc=<6.2198 0 0>, rot=<0 0 0 1>}
 dyn_loopFr_l_02 {parent=dyn_loopFr_l_01bis, loc=<5.0906 0 0>, rot=<0.0078 -0.0414 -0.161 -0.986>}
 dyn_loopFr_l_03 {parent=dyn_loopFr_l_02, loc=<4.1781 0 0>, rot=<0 0 0 1>}
 dyn_loopBk_l_04 {parent=dyn_loopFr_l_03, loc=<8.432 -0.1802 -4.3239>, rot=<-0.0649 -0.4816 -0.2117 -0.8479>}
 thigh_r {parent=pelvis, loc=<-2.5693 -0.0109 7.7501>, rot=<0.0572 -0.0012 0.9982 0.0155>}
 calf_r {parent=thigh_r, loc=<39.0484 0 0>, rot=<0.0006 -0.0161 -0.0391 -0.9991>}
 foot_r {parent=calf_r, loc=<42.561 -0.1077 -0.4292>, rot=<0.0418 -0.0455 0.0343 -0.9975>}
 ball_r {parent=foot_r, loc=<6.1599 -13.8264 -0.6806>, rot=<0.0001 -0.0001 -0.7063 -0.7079>}
 calf_twist_02_r {parent=calf_r, loc=<18.8064 0.0967 -0.0397>, rot=<0.0237 0.0095 -0.0035 -0.9997>}
 calf_twist_01_r {parent=calf_r, loc=<35.813 0.0967 -0.0397>, rot=<0.0237 0.0095 -0.0035 -0.9997>}
 deform_knee_in_base_r {parent=calf_r, loc=<-1.298 5.2957 0.0496>, rot=<0 0 0 -1>}
 deform_knee_in_r {parent=deform_knee_in_base_r, loc=<0 0 0>, rot=<0 0 0 1>}
 deform_knee_out_base_r {parent=calf_r, loc=<0 -4 0>, rot=<0 0 0 -1>}
 deform_knee_out_r {parent=deform_knee_out_base_r, loc=<0 0 0>, rot=<0 0 0 1>}
 thigh_twist_01_r {parent=thigh_r, loc=<21.6303 0 0>, rot=<0.0013 0 0 -1>}
 deform_groin_base_r {parent=thigh_r, loc=<2.1908 -8.1142 1.7442>, rot=<0 0 0 -1>}
 deform_groin_r {parent=deform_groin_base_r, loc=<0 0 0>, rot=<0 0 0 1>}
 dyn_fishBone_r_MSTR {parent=thigh_r, loc=<19.7378 2.3824 13.0291>, rot=<0 0 0 1>}
 dyn_fishBone_r_01 {parent=dyn_fishBone_r_MSTR, loc=<0 0 0>, rot=<0 0 0 1>}
 dyn_fork_r_MSTR {parent=thigh_r, loc=<12.3743 -13.1646 0.0444>, rot=<0.9838 -0.0214 -0.0508 0.1706>}
 dyn_fork_r_01 {parent=dyn_fork_r_MSTR, loc=<0 0 0>, rot=<0 0 0 1>}
 dyn_spoon_r_MSTR {parent=thigh_r, loc=<11.261 -13.6303 -2.6682>, rot=<0.997 -0.0069 -0.0571 0.0511>}
 dyn_spoon_r_01 {parent=dyn_spoon_r_MSTR, loc=<0 0 0>, rot=<0 0 0 1>}
 dyn_pouchThigh_r_MSTR {parent=pelvis, loc=<-14.5893 -13.5147 15.0464>, rot=<-0.0107 -0.4917 0.8698 0.0397>}
 dyn_pouchThigh_r_01 {parent=dyn_pouchThigh_r_MSTR, loc=<0.0003 0 0>, rot=<0 0 0 1>}
 deform_glute_base_l {parent=pelvis, loc=<-8.8064 -8.8222 -8.4287>, rot=<0.0275 0.0501 -0.4472 -0.8926>}
 deform_glute_l {parent=deform_glute_base_l, loc=<0 0 0>, rot=<0 0 0 1>}
 deform_glute_base_r {parent=pelvis, loc=<-8.8062 -8.8222 8.4287>, rot=<0.0501 -0.0275 0.8926 -0.4472>}
 deform_glute_r {parent=deform_glute_base_r, loc=<0 0 0>, rot=<0 0 0 1>}
 dyn_loopBk_l_MSTR {parent=pelvis, loc=<-13.278 -13.2467 -14.2774>, rot=<-0.1129 -0.9045 -0.2347 -0.3376>}
 dyn_loopBk_l_01 {parent=dyn_loopBk_l_MSTR, loc=<0 0 0>, rot=<0 0 0 1>}
 dyn_loopBk_l_02 {parent=dyn_loopBk_l_01, loc=<10.0009 -0.0213 -1.1345>, rot=<0.0417 -0.0142 -0.1279 -0.9908>}
]
variables [
 SplineIK_1_Output Spline {type=FControlRigSpline, default=()}
 beltSplineIK_10Val {type=FVector, default=(X=0,Y=0,Z=0)}
 beltSpline_ctrls {type="TArray<FRigElementKey>"}
 beltSpline_ctrls_pose {type=FRigPose, default=(Elements=,HierarchyTopologyVersion=-1,PoseHash=-1)}
 beltSpline_joints {type="TArray<FRigElementKey>"}
 spline_ctrls_bindPose {type=FRigPose, default=(Elements=,HierarchyTopologyVersion=-1,PoseHash=-1)}
 spline_ctrls_thigh_l_up_pose {type=FRigPose, default=(Elements=,HierarchyTopologyVersion=-1,PoseHash=-1)}
 spline_ctrls_thigh_r_up_pose {type=FRigPose, default=(Elements=,HierarchyTopologyVersion=-1,PoseHash=-1)}
 thigh_r_SPR_outputParam {type=float, default=0}
]
