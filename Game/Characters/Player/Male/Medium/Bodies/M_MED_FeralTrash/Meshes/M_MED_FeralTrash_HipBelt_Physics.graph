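# Physics asset for the hip belt of the medium male FeralTrash player body, with capsule bodies for its dangling trinkets and loops.
# Pelvis, lower spine and thigh capsules form the body the accessories hang from.
# Fork, spoon and fishbone each swing narrowly on a small master body joined to the right thigh with free rotation.
# Front and back left-thigh chains use swing limits and zero twist, linking through back segments 04 and 05 into a closed loop.
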PhysicsAsset /Game/Characters/Player/Male/Medium/Bodies/M_MED_FeralTrash/Meshes/M_MED_FeralTrash_HipBelt_Physics
bodies {
 dyn_fishBone_r_01 [{shapeType=Capsule, length=18.5, radius=3, shapeName=dyn_fishBone_r_01_capsule, center=<10.3846 0 2>, rotation=<rot 90 0 0>}]
 dyn_fishBone_r_MSTR [{shapeType=Capsule, length=0.505, radius=0.505, shapeName=dyn_fishBone_r_MSTR_capsule, rotation=<rot 0 -90 0>}]
 dyn_fork_r_01 [{shapeType=Capsule, length=16, radius=2, shapeName=dyn_fork_r_01_capsule, center=<10.3846 1 0>, rotation=<rot 90 0 0>}]
 dyn_fork_r_MSTR [{shapeType=Capsule, length=0.505, radius=0.505, shapeName=dyn_fork_r_MSTR_capsule, rotation=<rot 0 -90 0>}]
 dyn_loopBk_l_01 [{shapeType=Capsule, length=7, radius=2, shapeName=dyn_loopBk_l_01_capsule, center=<3.6187 0 0>, rotation=<rot 90 0 0>}]
 dyn_loopBk_l_02 [{shapeType=Capsule, length=7, radius=2, shapeName=dyn_loopBk_l_02_capsule, center=<3.6187 0 0>, rotation=<rot 90 0 0>}]
 dyn_loopBk_l_03 [{shapeType=Capsule, length=12, radius=2, shapeName=dyn_loopBk_l_03_capsule, center=<3.6187 0 0>, rotation=<rot 90 0 0>}]
 dyn_loopBk_l_04 [{shapeType=Capsule, length=6.5, radius=2, shapeName=dyn_loopBk_l_04_capsule, center=<4.0202 0 0>, rotation=<rot 90 0 0>}]
 dyn_loopBk_l_05 [{shapeType=Capsule, length=6.5, radius=2, shapeName=dyn_loopBk_l_05_capsule, center=<4.0202 0 0>, rotation=<rot 90 0 0>}]
 dyn_loopBk_l_MSTR [{shapeType=Capsule, length=7, radius=2, shapeName=dyn_loopBk_l_MSTR_capsule, center=<3.6187 0 0>, rotation=<rot 90 0 0>}]
 dyn_loopFr_l_01 [{shapeType=Capsule, length=7, radius=2, shapeName=dyn_loopFr_l_01_capsule, center=<6.0256 -2 0>, rotation=<rot 90 0 0>}]
 dyn_loopFr_l_02 [{shapeType=Capsule, length=7, radius=2, shapeName=dyn_loopFr_l_02_capsule, center=<3.6187 0 0>, rotation=<rot 90 0 0>}]
 dyn_loopFr_l_03 [{shapeType=Capsule, length=5.5, radius=3.5, shapeName=dyn_loopFr_l_03_capsule, center=<2.9712 -1 -1.5>, rotation=<rot 66.6831 -0.0001 0>}]
 dyn_loopFr_l_MSTR [{shapeType=Capsule, length=7, radius=2, shapeName=dyn_loopFr_l_MSTR_capsule, center=<3.6187 0 0>, rotation=<rot 90 0 0>}]
 dyn_spoon_r_01 [{shapeType=Capsule, length=16, radius=2, shapeName=dyn_spoon_r_01_capsule, center=<10.3846 1 0>, rotation=<rot 90 0 0>}]
 dyn_spoon_r_MSTR [{shapeType=Capsule, length=0.505, radius=0.505, shapeName=dyn_spoon_r_MSTR_capsule, rotation=<rot 0 -90 0>}]
 pelvis [{shapeType=Capsule, length=11.5, radius=14, shapeName=pelvis_capsule}]
 spine_01 [{shapeType=Capsule, length=7, radius=12.5, shapeName=spine_01_capsule, rotation=<rot 0 -90 0>}]
 thigh_l [{shapeType=Capsule, length=36.5, radius=10.5, shapeName=thigh_l_capsule, center=<-23.8981 0 -1.4114>, rotation=<rot 90 0 0>}]
 thigh_r [{shapeType=Capsule, length=35, radius=12, shapeName=thigh_l_capsule, center=<23.8981 0 1.4114>, rotation=<rot 90 180 0>}]
}
constraints [
 {bone1Name=spine_01, bone2Name=pelvis, swing1LimitDegrees=45, swing2LimitDegrees=45, twistLimitDegrees=45}
 {
  bone1Name=thigh_l
  bone2Name=pelvis
  swing1LimitDegrees=45
  swing1Motion=Free
  swing2LimitDegrees=45
  swing2Motion=Free
  twistLimitDegrees=45
  twistMotion=Free
 }
 {
  bone1Name=thigh_r
  bone2Name=pelvis
  swing1LimitDegrees=45
  swing1Motion=Free
  swing2LimitDegrees=45
  swing2Motion=Free
  twistLimitDegrees=45
  twistMotion=Free
 }
 {bone1Name=dyn_fork_r_01, bone2Name=dyn_fork_r_MSTR, swing1LimitDegrees=20, swing2LimitDegrees=20, twistLimitDegrees=0}
 {
  bone1Name=dyn_fork_r_MSTR
  bone2Name=thigh_r
  swing1LimitDegrees=45
  swing1Motion=Free
  swing2LimitDegrees=45
  swing2Motion=Free
  twistLimitDegrees=45
  twistMotion=Free
 }
 {
  bone1Name=dyn_spoon_r_MSTR
  bone2Name=thigh_r
  swing1LimitDegrees=45
  swing1Motion=Free
  swing2LimitDegrees=45
  swing2Motion=Free
  twistLimitDegrees=45
  twistMotion=Free
 }
 {bone1Name=dyn_spoon_r_01, bone2Name=dyn_spoon_r_MSTR, swing1LimitDegrees=20, swing2LimitDegrees=20, twistLimitDegrees=0}
 {
  bone1Name=dyn_fishBone_r_MSTR
  bone2Name=thigh_r
  swing1LimitDegrees=45
  swing1Motion=Free
  swing2LimitDegrees=45
  swing2Motion=Free
  twistLimitDegrees=45
  twistMotion=Free
 }
 {bone1Name=dyn_fishBone_r_01, bone2Name=dyn_fishBone_r_MSTR, swing1LimitDegrees=20, swing2LimitDegrees=20, twistLimitDegrees=0}
 {bone1Name=dyn_loopFr_l_MSTR, bone2Name=thigh_l, swing1LimitDegrees=30, swing2LimitDegrees=30, twistLimitDegrees=0}
 {bone1Name=dyn_loopFr_l_01, bone2Name=dyn_loopFr_l_MSTR, swing1LimitDegrees=30, swing2LimitDegrees=30, twistLimitDegrees=0}
 {bone1Name=dyn_loopFr_l_02, bone2Name=dyn_loopFr_l_01, swing1LimitDegrees=30, swing2LimitDegrees=30, twistLimitDegrees=0}
 {bone1Name=dyn_loopFr_l_03, bone2Name=dyn_loopFr_l_02, swing1LimitDegrees=30, swing2LimitDegrees=30, twistLimitDegrees=0}
 {bone1Name=dyn_loopBk_l_MSTR, bone2Name=thigh_l, swing1LimitDegrees=30, swing2LimitDegrees=30, twistLimitDegrees=0}
 {bone1Name=dyn_loopBk_l_01, bone2Name=dyn_loopBk_l_MSTR, swing1LimitDegrees=30, swing2LimitDegrees=30, twistLimitDegrees=0}
 {bone1Name=dyn_loopBk_l_02, bone2Name=dyn_loopBk_l_01, swing1LimitDegrees=30, swing2LimitDegrees=30, twistLimitDegrees=0}
 {bone1Name=dyn_loopBk_l_03, bone2Name=dyn_loopBk_l_02, swing1LimitDegrees=20, swing2LimitDegrees=20, twistLimitDegrees=0}
 {bone1Name=dyn_loopBk_l_04, bone2Name=dyn_loopFr_l_03, swing1LimitDegrees=20, swing2LimitDegrees=20, twistLimitDegrees=0}
 {bone1Name=dyn_loopBk_l_05, bone2Name=dyn_loopBk_l_04, swing1LimitDegrees=20, swing2LimitDegrees=20, twistLimitDegrees=0}
 {
  bone1Name=dyn_loopBk_l_05
  bone2Name=dyn_loopBk_l_03
  swing1LimitDegrees=45
  swing1Motion=Free
  swing2LimitDegrees=45
  swing2Motion=Free
  twistLimitDegrees=45
  twistMotion=Free
 }
]
