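# Physics asset for the medium male player FeralTrash body, giving capsule bodies and joints to its dangling vest panels and left-side loops.
# Pelvis, spine and thigh capsules form the body base; vest chains hang from spine_01 through MSTR anchors with free rotation and angle-limited child segments.
# The back loop hangs from the pelvis and the front from thigh_l, with the last back-loop bone constrained to the front chain's end.
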PhysicsAsset /Game/Characters/Player/Male/Medium/Bodies/M_MED_FeralTrash/Meshes/M_MED_FeralTrash_Physics
bodies {
 dyn_loopBk_l_01 [{shapeType=Capsule, length=7.2994, radius=2.6714, shapeName=dyn_loopBk_l_01_capsule, center=<4.7785 0 0>, rotation=<rot 90 0 0>}]
 dyn_loopBk_l_02 [{shapeType=Capsule, length=7.7857, radius=2.9261, shapeName=dyn_loopBk_l_02_capsule, center=<9.5105 0 0>, rotation=<rot 90 0 0>}]
 dyn_loopBk_l_04 [{shapeType=Capsule, length=9, radius=3.005, shapeName=dyn_loopBk_l_04_capsule, center=<4.4084 -1 -1.5>, rotation=<rot 90 0 0>}]
 dyn_loopBk_l_MSTR [{shapeType=Capsule, length=0.505, radius=0.505, shapeName=dyn_loopBk_l_MSTR_capsule, rotation=<rot 0 -90 0>}]
 dyn_loopFr_l_01 [{shapeType=Capsule, length=7, radius=2.5, shapeName=dyn_loopFr_l_01_capsule, center=<3.8655 0 0>, rotation=<rot 90 0 0>}]
 dyn_loopFr_l_03 [{shapeType=Capsule, length=10.5, radius=2.5, shapeName=dyn_loopFr_l_03_capsule, center=<3.8772 -3 -2>, rotation=<rot 67.8499 0 0>}]
 dyn_loopFr_l_MSTR [{shapeType=Capsule, length=12, radius=2.5, shapeName=dyn_loopFr_l_MSTR_capsule, center=<6.0435 0 0>, rotation=<rot 90 0 0>}]
 dyn_vestBk_c_01 [{shapeType=Capsule, length=12.5, radius=2.5, shapeName=dyn_vestBk_c_01_capsule, center=<5.8777 -2 0>, rotation=<rot 90 0 0>}]
 dyn_vestBk_c_MSTR [{shapeType=Capsule, length=0.505, radius=0.505, shapeName=dyn_vestBk_c_MSTR_capsule, rotation=<rot 0 -90 0>}]
 dyn_vestBk_l_01 [{shapeType=Capsule, length=12.5, radius=2.5, shapeName=dyn_vestBk_l_01_capsule, center=<5.8777 -2 0>, rotation=<rot 90 0 0>}]
 dyn_vestBk_l_MSTR [{shapeType=Capsule, length=0.505, radius=0.505, shapeName=dyn_vestBk_l_MSTR_capsule, rotation=<rot 0 -90 0>}]
 dyn_vestBk_r_01 [{shapeType=Capsule, length=12.5, radius=2.5, shapeName=dyn_vestBk_r_01_capsule, center=<5.8777 -2 0>, rotation=<rot 90 0 0>}]
 dyn_vestBk_r_MSTR [{shapeType=Capsule, length=0.505, radius=0.505, shapeName=dyn_vestBk_r_MSTR_capsule, rotation=<rot 0 -90 0>}]
 dyn_vestSide_l_01 [{shapeType=Capsule, length=15.5, radius=3, shapeName=dyn_vestSide_l_01_capsule, center=<5.8777 -3 0>, rotation=<rot 90 0 0>}]
 dyn_vestSide_l_MSTR [{shapeType=Capsule, length=0.505, radius=0.505, shapeName=dyn_vestSide_l_MSTR_capsule, rotation=<rot 0 -90 0>}]
 dyn_vestSide_r_01 [{shapeType=Capsule, length=12.5, radius=2.5, shapeName=dyn_vestSide_r_01_capsule, center=<5.8777 -4 0>, rotation=<rot 90 0 0>}]
 dyn_vestSide_r_MSTR [{shapeType=Capsule, length=0.505, radius=0.505, shapeName=dyn_vestSide_r_MSTR_capsule, rotation=<rot 0 -90 0>}]
 dyn_vest_l_01 [{shapeType=Capsule, length=10.5, radius=3, shapeName=dyn_vest_l_01_capsule, center=<4.8385 -4 0>, rotation=<rot 90 0 0>}]
 dyn_vest_l_02 [{shapeType=Capsule, length=18, radius=9.5, shapeName=dyn_vest_l_02_capsule, center=<5 -8 -7>, rotation=<rot 0.0001 -0.0001 -8.2532>}]
 dyn_vest_l_MSTR [{shapeType=Capsule, length=0.505, radius=0.505, shapeName=dyn_vest_l_MSTR_capsule, rotation=<rot 0 -90 0>}]
 dyn_vest_r_01 [{shapeType=Capsule, length=11, radius=2.5, shapeName=dyn_vest_r_01_capsule, center=<4.8385 -2 0>, rotation=<rot 90 0 0>}]
 dyn_vest_r_02 [{shapeType=Capsule, length=39, radius=8, shapeName=dyn_vest_r_02_capsule, center=<4.8385 -6 6>}]
 dyn_vest_r_MSTR [{shapeType=Capsule, length=0.505, radius=0.505, shapeName=dyn_vest_r_MSTR_capsule, rotation=<rot 0 -90 0>}]
 pelvis [
  {shapeType=Capsule, length=11.5, radius=14, shapeName=pelvis_capsule}
  {shapeType=Capsule, length=14.5, radius=14, shapeName=pelvis_capsule, center=<-6.3494 -2.4115 0>}
 ]
 spine_01 [{shapeType=Capsule, length=7, radius=12.5, shapeName=spine_01_capsule, rotation=<rot 0 -90 0>}]
 thigh_l [{shapeType=Capsule, length=47, radius=13, shapeName=thigh_l_capsule, center=<-23.8981 0 0>, rotation=<rot 90 0 0>}]
 thigh_r [{shapeType=Capsule, length=47, radius=13, shapeName=thigh_l_capsule, center=<23.8981 0 0>, rotation=<rot 90 180 0>}]
}
constraints [
 {
  bone1Name=thigh_l
  bone2Name=pelvis
  swing1LimitDegrees=45
  swing1Motion=Free
  swing2LimitDegrees=45
  swing2Motion=Free
  twistLimitDegrees=45
  twistMotion=Free
 }
 {
  bone1Name=dyn_vestBk_c_MSTR
  bone2Name=spine_01
  swing1LimitDegrees=45
  swing1Motion=Free
  swing2LimitDegrees=45
  swing2Motion=Free
  twistLimitDegrees=45
  twistMotion=Free
 }
 {bone1Name=dyn_vestBk_c_01, bone2Name=dyn_vestBk_c_MSTR, swing1LimitDegrees=20, swing2LimitDegrees=20, twistLimitDegrees=0}
 {
  bone1Name=dyn_vestBk_l_MSTR
  bone2Name=spine_01
  swing1LimitDegrees=45
  swing1Motion=Free
  swing2LimitDegrees=45
  swing2Motion=Free
  twistLimitDegrees=45
  twistMotion=Free
 }
 {bone1Name=dyn_vestBk_l_01, bone2Name=dyn_vestBk_l_MSTR, swing1LimitDegrees=20, swing2LimitDegrees=20, twistLimitDegrees=0}
 {
  bone1Name=dyn_vestBk_r_MSTR
  bone2Name=spine_01
  swing1LimitDegrees=45
  swing1Motion=Free
  swing2LimitDegrees=45
  swing2Motion=Free
  twistLimitDegrees=45
  twistMotion=Free
 }
 {bone1Name=dyn_vestBk_r_01, bone2Name=dyn_vestBk_r_MSTR, swing1LimitDegrees=20, swing2LimitDegrees=20, twistLimitDegrees=0}
 {
  bone1Name=dyn_vestSide_l_MSTR
  bone2Name=spine_01
  swing1LimitDegrees=45
  swing1Motion=Free
  swing2LimitDegrees=45
  swing2Motion=Free
  twistLimitDegrees=45
  twistMotion=Free
 }
 {bone1Name=dyn_vestSide_l_01, bone2Name=dyn_vestSide_l_MSTR, swing1LimitDegrees=20, swing2LimitDegrees=20, twistLimitDegrees=0}
 {
  bone1Name=dyn_vestSide_r_MSTR
  bone2Name=spine_01
  swing1LimitDegrees=45
  swing1Motion=Free
  swing2LimitDegrees=45
  swing2Motion=Free
  twistLimitDegrees=45
  twistMotion=Free
 }
 {bone1Name=dyn_vestSide_r_01, bone2Name=dyn_vestSide_r_MSTR, swing1LimitDegrees=12, swing2LimitDegrees=12, twistLimitDegrees=0}
 {
  bone1Name=dyn_vest_l_MSTR
  bone2Name=spine_01
  swing1LimitDegrees=45
  swing1Motion=Free
  swing2LimitDegrees=45
  swing2Motion=Free
  twistLimitDegrees=45
  twistMotion=Free
 }
 {bone1Name=dyn_vest_l_01, bone2Name=dyn_vest_l_MSTR, swing1LimitDegrees=55, swing2LimitDegrees=20, twistLimitDegrees=0}
 {bone1Name=dyn_vest_l_02, bone2Name=dyn_vest_l_01, swing1LimitDegrees=55, swing2LimitDegrees=20, twistLimitDegrees=30}
 {
  bone1Name=dyn_vest_r_MSTR
  bone2Name=spine_01
  swing1LimitDegrees=45
  swing1Motion=Free
  swing2LimitDegrees=45
  swing2Motion=Free
  twistLimitDegrees=45
  twistMotion=Free
 }
 {bone1Name=dyn_vest_r_01, bone2Name=dyn_vest_r_MSTR, swing1LimitDegrees=55, swing2LimitDegrees=15, twistLimitDegrees=0}
 {bone1Name=dyn_vest_r_02, bone2Name=dyn_vest_r_01, swing1LimitDegrees=70, swing2LimitDegrees=20, twistLimitDegrees=0}
 {
  bone1Name=spine_01
  bone2Name=pelvis
  swing1LimitDegrees=45
  swing1Motion=Free
  swing2LimitDegrees=45
  swing2Motion=Free
  twistLimitDegrees=45
  twistMotion=Free
 }
 {
  bone1Name=dyn_loopBk_l_MSTR
  bone2Name=pelvis
  swing1LimitDegrees=45
  swing1Motion=Free
  swing2LimitDegrees=45
  swing2Motion=Free
  twistLimitDegrees=45
  twistMotion=Free
 }
 {bone1Name=dyn_loopBk_l_01, bone2Name=dyn_loopBk_l_MSTR, swing1LimitDegrees=20, swing2LimitDegrees=20, twistLimitDegrees=0}
 {bone1Name=dyn_loopBk_l_02, bone2Name=dyn_loopBk_l_01, swing1LimitDegrees=20, swing2LimitDegrees=20, twistLimitDegrees=0}
 {
  bone1Name=dyn_loopFr_l_01
  bone2Name=dyn_loopFr_l_MSTR
  swing1LimitDegrees=45
  swing1Motion=Free
  swing2LimitDegrees=45
  swing2Motion=Free
  twistLimitDegrees=45
  twistMotion=Free
 }
 {
  bone1Name=dyn_loopFr_l_MSTR
  bone2Name=thigh_l
  swing1LimitDegrees=45
  swing1Motion=Free
  swing2LimitDegrees=45
  swing2Motion=Free
  twistLimitDegrees=45
  twistMotion=Free
 }
 {bone1Name=dyn_loopFr_l_03, bone2Name=dyn_loopFr_l_01, swing1LimitDegrees=45, swing2LimitDegrees=0, twistLimitDegrees=1}
 {bone1Name=dyn_loopBk_l_04, bone2Name=dyn_loopFr_l_03, swing1LimitDegrees=45, swing2LimitDegrees=45, twistLimitDegrees=45}
 {
  bone1Name=thigh_r
  bone2Name=pelvis
  swing1LimitDegrees=45
  swing1Motion=Free
  swing2LimitDegrees=45
  swing2Motion=Free
  twistLimitDegrees=45
  twistMotion=Free
 }
]
