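# Post-process style AnimBlueprint for the FeralTrash male body, adding secondary motion to collar, buckles, pouches and dangling trash.
# Copied parent pose feeds Control Rig, weighted constraints, additive collar transforms, then RigidBody and AnimDynamics, before bottle-cap constraints reach output.
# Cached poses chain stages; the vest-offset cache and LookAt nodes never reach output, and hip-belt and bottle simulations have zero alpha.
# EventGraph computes head-rotation alphas, gravity override, and cape-dependent tail visibility.
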AnimBlueprint /Game/Characters/Player/Male/Medium/Bodies/M_MED_FeralTrash/Meshes/M_MED_FeralTrash_AnimBP
target_skeleton=&/M_MED_FeralTrash_Skeleton
graphs [
 AnimGraph {nodes [
   Look At - Bone: dyn_bottleCap_06(Alpha=1) {
    alphaBoolBlend {blendOption=Linear}
    boneToModify {boneName=dyn_bottleCap_06}
    lookAtLocation {x=0, z=50}
    lookAtTarget {boneReference {boneName=dyn_loopBk_l_02}}
    lookAt_Axis {axis=<1 0>}
    lookUp_Axis {axis {x=0, z=-1}}
   }
   n1 = "Constraint - dyn_bottleCap_01 (PPPPP)"!(
    ComponentPose
    Alpha=1
    ConstraintWeights_0=0.8074
    ConstraintWeights_1:float=0.1069
    ConstraintWeights_2:float=0.075
    ConstraintWeights_3:float=0.0095
    ConstraintWeights_4:float=0.0011
   ) -> (Pose) {
    alphaBoolBlend {blendOption=Linear}
    boneToModify {boneName=dyn_bottleCap_01}
    constraintSetup [
     {targetBone {boneName=dyn_loopFr_l_03}}
     {targetBone {boneName=dyn_loopFr_l_02}}
     {targetBone {boneName=dyn_loopBk_l_04}}
     {targetBone {boneName=dyn_loopBk_l_02}}
     {targetBone {boneName=dyn_loopBk_l_01}}
    ]
    constraintWeights [1,1,1,1,1]
   }
   Look At - Bone: dyn_bottleCap_01(ComponentPose=@n1, Alpha=1) {
    alphaBoolBlend {blendOption=Linear}
    boneToModify {boneName=dyn_bottleCap_01}
    lookAtTarget {boneReference {boneName=dyn_bottleCap_02}}
    lookAt_Axis {axis=<1 0>}
   }
   Look At - Bone: dyn_bottleCap_02(Alpha=1) {
    alphaBoolBlend {blendOption=Linear}
    boneToModify {boneName=dyn_bottleCap_02}
    lookAtLocation {x=50, z=50}
    lookAtTarget {boneReference {boneName=dyn_bottleCap_03}}
    lookAt_Axis {axis=<1 0>}
    lookUp_Axis {axis {x=0, z=-1}}
   }
   Look At - Bone: dyn_bottleCap_03(Alpha=1) {
    alphaBoolBlend {blendOption=Linear}
    boneToModify {boneName=dyn_bottleCap_03}
    lookAtLocation {x=50}
    lookAtTarget {boneReference {boneName=dyn_bottleCap_04}}
    lookAt_Axis {axis=<1 0>}
    lookUp_Axis {axis {x=0, z=-1}}
   }
   Look At - Bone: dyn_bottleCap_04(Alpha=1) {
    alphaBoolBlend {blendOption=Linear}
    boneToModify {boneName=dyn_bottleCap_04}
    lookAtLocation {x=50}
    lookAtTarget {boneReference {boneName=dyn_bottleCap_05}}
    lookAt_Axis {axis=<1 0>}
    lookUp_Axis {axis {x=0, z=-1}}
   }
   Look At - Bone: dyn_bottleCap_05(Alpha=1) {
    alphaBoolBlend {blendOption=Linear}
    boneToModify {boneName=dyn_bottleCap_05}
    lookAtLocation {x=50}
    lookAtTarget {boneReference {boneName=dyn_bottleCap_06}}
    lookAt_Axis {axis=<1 0>}
    lookUp_Axis {axis {x=0, z=-1}}
   }
   n2 = "Constraint - dyn_bottleCap_02 (PPPPP)"!(
    ComponentPose=Local To Component(LocalPose=Use cached pose 'dynamicsCache'())
    Alpha=1
    ConstraintWeights_0=0.705
    ConstraintWeights_1:float=0.2085
    ConstraintWeights_2:float=0.0708
    ConstraintWeights_3:float=0.009
    ConstraintWeights_4:float=0.0068
   ) -> (Pose) {
    alphaBoolBlend {blendOption=Linear}
    boneToModify {boneName=dyn_bottleCap_02}
    constraintSetup [
     {targetBone {boneName=dyn_loopFr_l_03}}
     {targetBone {boneName=dyn_loopBk_l_04}}
     {targetBone {boneName=dyn_loopBk_l_02}}
     {targetBone {boneName=dyn_loopBk_l_01}}
     {targetBone {boneName=dyn_loopFr_l_02}}
    ]
    constraintWeights [1,1,1,1,1]
   }
   n3 = "Constraint - dyn_bottleCap_03 (PPPP)"!(
    ComponentPose=@n2
    Alpha=1
    ConstraintWeights_0=0.4584
    ConstraintWeights_1:float=0.299
    ConstraintWeights_2:float=0.2112
    ConstraintWeights_3:float=0.0314
   ) -> (Pose) {
    alphaBoolBlend {blendOption=Linear}
    boneToModify {boneName=dyn_bottleCap_03}
    constraintSetup [
     {targetBone {boneName=dyn_loopFr_l_03}}
     {targetBone {boneName=dyn_loopBk_l_04}}
     {targetBone {boneName=dyn_loopBk_l_02}}
     {targetBone {boneName=dyn_loopBk_l_01}}
    ]
    constraintWeights [1,1,1,1]
   }
   n4 = "Constraint - dyn_bottleCap_04 (PPPP)"!(
    ComponentPose=@n3
    Alpha=1
    ConstraintWeights_0=0.3826
    ConstraintWeights_1:float=0.2908
    ConstraintWeights_2:float=0.2554
    ConstraintWeights_3:float=0.0713
   ) -> (Pose) {
    alphaBoolBlend {blendOption=Linear}
    boneToModify {boneName=dyn_bottleCap_04}
    constraintSetup [
     {targetBone {boneName=dyn_loopBk_l_02}}
     {targetBone {boneName=dyn_loopBk_l_04}}
     {targetBone {boneName=dyn_loopFr_l_03}}
     {targetBone {boneName=dyn_loopBk_l_01}}
    ]
    constraintWeights [1,1,1,1]
   }
   n5 = "Constraint - dyn_bottleCap_05 (PPPP)"!(
    ComponentPose=@n4
    Alpha=1
    ConstraintWeights_0=0.5586
    ConstraintWeights_1:float=0.2038
    ConstraintWeights_2:float=0.1443
    ConstraintWeights_3:float=0.0933
   ) -> (Pose) {
    alphaBoolBlend {blendOption=Linear}
    boneToModify {boneName=dyn_bottleCap_05}
    constraintSetup [
     {targetBone {boneName=dyn_loopBk_l_02}}
     {targetBone {boneName=dyn_loopBk_l_04}}
     {targetBone {boneName=dyn_loopBk_l_01}}
     {targetBone {boneName=dyn_loopFr_l_03}}
    ]
    constraintWeights [1,1,1,1]
   }
   n6 = "Constraint - dyn_bottleCap_06 (PPPP)"!(
    ComponentPose=@n5
    Alpha=1
    ConstraintWeights_0=0.6041
    ConstraintWeights_1:float=0.302
    ConstraintWeights_2:float=0.0763
    ConstraintWeights_3:float=0.0176
   ) -> (Pose) {
    alphaBoolBlend {blendOption=Linear}
    boneToModify {boneName=dyn_bottleCap_06}
    constraintSetup [
     {targetBone {boneName=dyn_loopBk_l_02}}
     {targetBone {boneName=dyn_loopBk_l_01}}
     {targetBone {boneName=dyn_loopBk_l_04}}
     {targetBone {boneName=dyn_loopFr_l_03}}
    ]
    constraintWeights [1,1,1,1]
   }
   Output Pose(Result=Component To Local(ComponentPose=@n6)) {layerGroup=None}
   n7 = Control Rig(Source=Copy Pose From Mesh() {bUseAttachedParent=true}) {
    alphaBoolBlend {blendOption=Linear}
    controlRigAssetReference {blueprintRigClass=&/M_MED_FeralTrash_CtrlRig.M_MED_FeralTrash_CtrlRig_C}
   }
   Save cached pose 'CRCache'(Pose=@n7) {cacheName=CRCache}
   n8 = RigidBody(ComponentPose=Local To Component(LocalPose=Use cached pose 'preSim1Cache'())) {
    bClampLinearTranslationLimitToRefPose=true
    bOverrideWorldGravity=true
    overridePhysicsAsset=&/M_MED_FeralTrash_Physics
    simulationSpace=BaseBoneSpace
    alphaBoolBlend {blendOption=Linear}
    baseBoneRef {boneName=pelvis}
    componentAppliedLinearAccClamp=<6000 6000 6000>
    componentLinearAccScale=<0.1 0.3 0.3>
    componentLinearVelScale=<0.1 0.3 0.3>
    overrideWorldGravity {z=-980}
   }
   n9 = RigidBody(ComponentPose=@n8, Alpha=0) {
    bClampLinearTranslationLimitToRefPose=true
    bOverrideWorldGravity=true
    overridePhysicsAsset=&/M_MED_FeralTrash_HipBelt_Physics
    simulationSpace=BaseBoneSpace
    alphaBoolBlend {blendOption=Linear}
    baseBoneRef {boneName=thigh_l}
    componentAppliedLinearAccClamp=<6000 6000 6000>
    componentLinearAccScale=<0.2 0.2 0.2>
    componentLinearVelScale=<0.2 0.2 0.2>
    overrideWorldGravity {z=-980}
   }
   n10 = Anim Dynamics - Bone: dyn_fishBone_r_01(ComponentPose=@n9, Alpha=1, ExternalForce="0,0,0") {
    angularDampingOverride=1
    angularSpringConstant=3000
    bAngularSpring=true
    bOverrideAngularDamping=true
    bUseGravityOverride=true
    gravityScale=0.35
    simulationSpace=BoneRelative
    alphaBoolBlend {blendOption=Linear}
    boundBone {boneName=dyn_fishBone_r_01}
    componentAppliedLinearAccClamp=<6000 6000 6000>
    componentLinearAccScale=<1 1 1>
    componentLinearVelScale=<1 1 1>
    gravityOverride {z=-980}
    physicsBodyDefinitions [{
      boundBone {boneName=dyn_fishBone_r_01}
      boxExtents=<20 30 30>
      constraintSetup {angularLimitsMax=<15 0 90>, angularLimitsMin=<-15 -45 -90>, angularTarget=<1 0 0>}
      localJointOffset=<10 0 0>
     }]
    relativeSpaceBone {boneName=root}
   }
   n11 = Anim Dynamics - Bone: dyn_fork_r_01(ComponentPose=@n10, Alpha=1, ExternalForce="0,0,0") {
    angularDampingOverride=1
    angularSpringConstant=20000
    bOverrideAngularDamping=true
    bUseGravityOverride=true
    gravityScale=0.35
    simulationSpace=BoneRelative
    alphaBoolBlend {blendOption=Linear}
    boundBone {boneName=dyn_fork_r_01}
    componentAppliedLinearAccClamp=<6000 6000 6000>
    componentLinearAccScale=<0.6 0.6 0.6>
    componentLinearVelScale=<0.6 0.6 0.6>
    gravityOverride {z=-980}
    physicsBodyDefinitions [{
      boundBone {boneName=dyn_fork_r_01}
      boxExtents=<20 30 30>
      constraintSetup {angularLimitsMax=<15 45 45>, angularLimitsMin=<-15 -45 8>, angularTarget=<1 0 0>}
      localJointOffset=<6 0 0>
     }]
    relativeSpaceBone {boneName=root}
   }
   n12 = Anim Dynamics - Bone: dyn_spoon_r_01(ComponentPose=@n11, Alpha=1, ExternalForce="0,0,0") {
    angularDampingOverride=1
    angularSpringConstant=20000
    bOverrideAngularDamping=true
    bUseGravityOverride=true
    gravityScale=0.35
    simulationSpace=BoneRelative
    alphaBoolBlend {blendOption=Linear}
    boundBone {boneName=dyn_spoon_r_01}
    componentAppliedLinearAccClamp=<6000 6000 6000>
    componentLinearAccScale=<1 1 1>
    componentLinearVelScale=<1 1 1>
    gravityOverride {z=-980}
    physicsBodyDefinitions [{
      boundBone {boneName=dyn_spoon_r_01}
      boxExtents=<20 30 30>
      constraintSetup {angularLimitsMax=<15 30 45>, angularLimitsMin=<-15 -30 -2>, angularTarget=<1 0 0>}
      localJointOffset=<4 0 0>
     }]
    relativeSpaceBone {boneName=root}
   }
   n13 = Anim Dynamics - Bone: dyn_bottle_l_01(ComponentPose=@n12, Alpha=0, ExternalForce="0,0,0") {
    angularDampingOverride=1
    angularSpringConstant=20000
    bOverrideAngularDamping=true
    bUseGravityOverride=true
    gravityScale=0.35
    simulationSpace=BoneRelative
    alphaBoolBlend {blendOption=Linear}
    boundBone {boneName=dyn_bottle_l_01}
    componentLinearAccScale=<1 1 1>
    componentLinearVelScale=<1 1 1>
    gravityOverride {z=-980}
    physicsBodyDefinitions [{
      boundBone {boneName=dyn_bottle_l_01}
      boxExtents=<20 30 30>
      constraintSetup {angularLimitsMax=<15 5 15>, angularLimitsMin=<-15 -5 1>, angularTarget=<1 0 0>}
      localJointOffset=<4 0 0>
     }]
    relativeSpaceBone {boneName=spine_02}
   }
   Save cached pose 'dynamicsCache'(Pose=Component To Local(ComponentPose=@n13)) {cacheName=dynamicsCache}
   n14 = "Constraint - dyn_loopBk_l_MSTR (PPPPPPPP)"!(
    ComponentPose=Local To Component(LocalPose=Use cached pose 'CRCache'())
    Alpha=1
    ConstraintWeights_0=0.3184
    ConstraintWeights_1:float=0.3078
    ConstraintWeights_2:float=0.2964
    ConstraintWeights_3:float=0.0392
    ConstraintWeights_4:float=0.0149
    ConstraintWeights_5:float=0.0146
    ConstraintWeights_6:float=0.0054
    ConstraintWeights_7:float=0.0033
   ) -> (Pose) {
    alphaBoolBlend {blendOption=Linear}
    boneToModify {boneName=dyn_loopBk_l_MSTR}
    constraintSetup [
     {targetBone {boneName=pelvis}}
     {targetBone {boneName=deform_glute_l}}
     {targetBone {boneName=thigh_twist_01_l}}
     {targetBone {boneName=dyn_pouchThigh_r_MSTR}}
     {targetBone {boneName=deform_glute_r}}
     {targetBone {boneName=thigh_twist_01_r}}
     {targetBone {boneName=deform_groin_l}}
     {targetBone {boneName=spine_02}}
    ]
    constraintWeights [1,1,1,1,1,1,1,1]
   }
   n15 = "Constraint - dyn_fishBone_r_MSTR (PPPPPPPP)"!(
    ComponentPose=@n14
    Alpha=1
    ConstraintWeights_0=0.7687
    ConstraintWeights_1:float=0.0828
    ConstraintWeights_2:float=0.0798
    ConstraintWeights_3:float=0.0473
    ConstraintWeights_4:float=0.0101
    ConstraintWeights_5:float=0.0072
    ConstraintWeights_6:float=0.0031
    ConstraintWeights_7:float=0.0009
   ) -> (Pose) {
    alphaBoolBlend {blendOption=Linear}
    boneToModify {boneName=dyn_fishBone_r_MSTR}
    constraintSetup [
     {targetBone {boneName=dyn_fishBone_r_MSTR}}
     {targetBone {boneName=dyn_pouchThigh_r_MSTR}}
     {targetBone {boneName=thigh_twist_01_r}}
     {targetBone {boneName=pelvis}}
     {targetBone {boneName=thigh_r}}
     {targetBone {boneName=deform_groin_r}}
     {targetBone {boneName=deform_glute_r}}
     {targetBone {boneName=deform_glute_l}}
    ]
    constraintWeights [1,1,1,1,1,1,1,1]
   }
   n16 = "Constraint - dyn_shoulderBuckle_l_MSTR (PPPPPPPP)"!(
    ComponentPose=@n15
    Alpha=1
    ConstraintWeights_0=0.1915
    ConstraintWeights_1:float=0.1842
    ConstraintWeights_2:float=0.1801
    ConstraintWeights_3:float=0.1648
    ConstraintWeights_4:float=0.1194
    ConstraintWeights_5:float=0.1178
    ConstraintWeights_6:float=0.0242
    ConstraintWeights_7:float=0.0182
   ) -> (Pose) {
    alphaBoolBlend {blendOption=Linear}
    boneToModify {boneName=dyn_shoulderBuckle_l_MSTR}
    constraintSetup [
     {targetBone {boneName=upperarm_twist_01_l}}
     {targetBone {boneName=clavicle_l}}
     {targetBone {boneName=deform_pec_l}}
     {targetBone {boneName=deform_upperarm_fr_l}}
     {targetBone {boneName=spine_04}}
     {targetBone {boneName=spine_03}}
     {targetBone {boneName=spine_02}}
     {targetBone {boneName=upperarm_twist_02_l}}
    ]
    constraintWeights [1,1,1,1,1,1,1,1]
   }
   n17 = "Constraint - dyn_pouchThigh_r_MSTR (PPPPPPPP)"!(
    ComponentPose=@n16
    Alpha=1
    ConstraintWeights_0=0.6426
    ConstraintWeights_1:float=0.1787
    ConstraintWeights_2:float=0.0915
    ConstraintWeights_3:float=0.0601
    ConstraintWeights_4:float=0.0102
    ConstraintWeights_5:float=0.0068
    ConstraintWeights_6:float=0.0059
    ConstraintWeights_7:float=0.0043
   ) -> (Pose) {
    alphaBoolBlend {blendOption=Linear}
    boneToModify {boneName=dyn_pouchThigh_r_MSTR}
    constraintSetup [
     {targetBone {boneName=dyn_pouchThigh_r_MSTR}}
     {targetBone {boneName=dyn_fishBone_r_MSTR}}
     {targetBone {boneName=pelvis}}
     {targetBone {boneName=thigh_twist_01_r}}
     {targetBone {boneName=deform_glute_r}}
     {targetBone {boneName=thigh_r}}
     {targetBone {boneName=deform_glute_l}}
     {targetBone {boneName=thigh_twist_01_l}}
    ]
    constraintWeights [1,1,1,1,1,1,1,1]
   }
   Save cached pose 'rivetCache'(Pose=Component To Local(ComponentPose=@n17)) {cacheName=rivetCache}
   n18 = "Constraint - dyn_collarA_l_MSTR (P)"(ComponentPose=Local To Component(LocalPose=Use cached pose 'rivetCache'()), Alpha=0.8, ConstraintWeights_0=1) {
    alphaBoolBlend {blendOption=Linear}
    boneToModify {boneName=dyn_collarA_l_MSTR}
    constraintSetup [{targetBone {boneName=clavicle_l}}]
    constraintWeights [1]
   }
   n19 = "Constraint - dyn_collarA_r_MSTR (P)"(ComponentPose=@n18, Alpha=0.8, ConstraintWeights_0=1) {
    alphaBoolBlend {blendOption=Linear}
    boneToModify {boneName=dyn_collarA_r_MSTR}
    constraintSetup [{targetBone {boneName=clavicle_r}}]
    constraintWeights [1]
   }
   n20 = "Transform (Modify) Bone - Bone: dyn_collar_c_01"(ComponentPose=@n19, Alpha=1, Translation="0,0,-2", Rotation="-80,0,0", Scale="1,1,1") {
    rotationMode=BMM_Additive
    rotationSpace=BCS_BoneSpace
    translationMode=BMM_Additive
    translationSpace=BCS_BoneSpace
    alphaBoolBlend {blendOption=Linear}
    boneToModify {boneName=dyn_collar_c_01}
    rotation {pitch=-80}
    translation {z=-2}
   }
   n21 = "Transform (Modify) Bone - Bone: dyn_collarB_l_01"(ComponentPose=@n20, Alpha=1, Translation="0,0,-2", Rotation="-60,0,0", Scale="1,1,1") {
    rotationMode=BMM_Additive
    rotationSpace=BCS_BoneSpace
    translationMode=BMM_Additive
    translationSpace=BCS_BoneSpace
    alphaBoolBlend {blendOption=Linear}
    boneToModify {boneName=dyn_collarB_l_01}
    rotation {pitch=-60}
    translation {z=-2}
   }
   n22 = "Transform (Modify) Bone - Bone: dyn_collarB_r_01"(ComponentPose=@n21, Alpha=1, Translation="-2,0,-2", Rotation="-80,0,0", Scale="1,1,1") {
    rotationMode=BMM_Additive
    rotationSpace=BCS_BoneSpace
    translationMode=BMM_Additive
    translationSpace=BCS_BoneSpace
    alphaBoolBlend {blendOption=Linear}
    boneToModify {boneName=dyn_collarB_r_01}
    rotation {pitch=-80}
    translation {x=-2, z=-2}
   }
   n23 = "Transform (Modify) Bone - Bone: dyn_collar_c_01"(ComponentPose=@n22, Alpha=1, Translation="0,0,0", Rotation="-40,0,0", Scale="1,1,1") {
    rotationMode=BMM_Additive
    rotationSpace=BCS_BoneSpace
    translationMode=BMM_Additive
    translationSpace=BCS_BoneSpace
    alphaBoolBlend {blendOption=Linear}
    boneToModify {boneName=dyn_collar_c_01}
    rotation {pitch=-40}
   }
   n24 = "Transform (Modify) Bone - Bone: dyn_collarB_l_01"(ComponentPose=@n23, Alpha=1, Translation="0,0,0", Rotation="0,0,0", Scale="1,1,1") {
    rotationMode=BMM_Additive
    rotationSpace=BCS_BoneSpace
    translationMode=BMM_Additive
    translationSpace=BCS_BoneSpace
    alphaBoolBlend {blendOption=Linear}
    boneToModify {boneName=dyn_collarB_l_01}
   }
   n25 = "Transform (Modify) Bone - Bone: dyn_collarB_r_01"(ComponentPose=@n24, Alpha=1, Translation="0,0,0", Rotation="-40,0,0", Scale="1,1,1") {
    rotationMode=BMM_Additive
    rotationSpace=BCS_BoneSpace
    translationMode=BMM_Additive
    translationSpace=BCS_BoneSpace
    alphaBoolBlend {blendOption=Linear}
    boneToModify {boneName=dyn_collarB_r_01}
    rotation {pitch=-40}
   }
   n26 = "Transform (Modify) Bone - Bone: dyn_collar_c_01"(ComponentPose=@n25, Alpha=1, Translation="0,0,0", Rotation="-42,0,0", Scale="1,1,1") {
    rotationMode=BMM_Additive
    rotationSpace=BCS_BoneSpace
    translationMode=BMM_Additive
    translationSpace=BCS_BoneSpace
    alphaBoolBlend {blendOption=Linear}
    boneToModify {boneName=dyn_collar_c_01}
    rotation {pitch=-42}
   }
   n27 = "Transform (Modify) Bone - Bone: dyn_collarB_l_01"(ComponentPose=@n26, Alpha=1, Translation="0,0,0", Rotation="-40,0,0", Scale="1,1,1") {
    rotationMode=BMM_Additive
    rotationSpace=BCS_BoneSpace
    translationMode=BMM_Additive
    translationSpace=BCS_BoneSpace
    alphaBoolBlend {blendOption=Linear}
    boneToModify {boneName=dyn_collarB_l_01}
    rotation {pitch=-40}
   }
   n28 = "Transform (Modify) Bone - Bone: dyn_collarB_r_01"(ComponentPose=@n27, Alpha=1, Translation="0,0,0", Rotation="0,0,0", Scale="1,1,1") {
    rotationMode=BMM_Additive
    rotationSpace=BCS_BoneSpace
    translationMode=BMM_Additive
    translationSpace=BCS_BoneSpace
    alphaBoolBlend {blendOption=Linear}
    boneToModify {boneName=dyn_collarB_r_01}
   }
   n29 = "Transform (Modify) Bone - Bone: dyn_collar_c_01"(ComponentPose=@n28, Alpha=1, Translation="0,0,0", Rotation="20,0,-30", Scale="1,1,1") {
    rotationMode=BMM_Additive
    rotationSpace=BCS_BoneSpace
    translationMode=BMM_Additive
    translationSpace=BCS_BoneSpace
    alphaBoolBlend {blendOption=Linear}
    boneToModify {boneName=dyn_collar_c_01}
    rotation {pitch=20, roll=-30}
   }
   n30 = "Transform (Modify) Bone - Bone: dyn_collarB_l_01"(ComponentPose=@n29, Alpha=1, Translation="0,0,-4", Rotation="-30,0,0", Scale="1,1,1") {
    rotationMode=BMM_Additive
    rotationSpace=BCS_BoneSpace
    translationMode=BMM_Additive
    translationSpace=BCS_BoneSpace
    alphaBoolBlend {blendOption=Linear}
    boneToModify {boneName=dyn_collarB_l_01}
    rotation {pitch=-30}
    translation {z=-4}
   }
   n31 = "Transform (Modify) Bone - Bone: dyn_collarB_r_01"(ComponentPose=@n30, Alpha=1, Translation="0,0,0", Rotation="-40,0,-20", Scale="1,1,1") {
    rotationMode=BMM_Additive
    rotationSpace=BCS_BoneSpace
    translationMode=BMM_Additive
    translationSpace=BCS_BoneSpace
    alphaBoolBlend {blendOption=Linear}
    boneToModify {boneName=dyn_collarB_r_01}
    rotation {pitch=-40, roll=-20}
   }
   n32 = Anim Dynamics - Bone: dyn_collarA_l_01(ComponentPose=@n31, Alpha=1, ExternalForce="0,0,0") {
    angularDampingOverride=0.99
    angularSpringConstant=6000
    bAngularSpring=true
    bOverrideAngularDamping=true
    bUseGravityOverride=true
    gravityScale=0.35
    simulationSpace=BoneRelative
    alphaBoolBlend {blendOption=Linear}
    boundBone {boneName=dyn_collarA_l_01}
    componentAppliedLinearAccClamp=<6000 6000 6000>
    componentLinearAccScale=<0.1 0.1 0.1>
    componentLinearVelScale=<0.1 0.1 0.1>
    gravityOverride {z=-980}
    physicsBodyDefinitions [{
      boundBone {boneName=dyn_collarA_l_01}
      boxExtents=<20 30 30>
      constraintSetup {angularLimitsMax=<0 5 0>, angularLimitsMin=<0 -5 0>, angularTarget=<1 0 0>}
      localJointOffset=<4 0 0>
     }]
    relativeSpaceBone {boneName=root}
   }
   n33 = Anim Dynamics - Bone: dyn_collarA_r_01(ComponentPose=@n32, Alpha=1, ExternalForce="0,0,0") {
    angularDampingOverride=0.99
    angularSpringConstant=6000
    bAngularSpring=true
    bOverrideAngularDamping=true
    bUseGravityOverride=true
    gravityScale=0.35
    simulationSpace=BoneRelative
    alphaBoolBlend {blendOption=Linear}
    boundBone {boneName=dyn_collarA_r_01}
    componentAppliedLinearAccClamp=<6000 6000 6000>
    componentLinearAccScale=<0.1 0.1 0.1>
    componentLinearVelScale=<0.1 0.1 0.1>
    gravityOverride {z=-980}
    physicsBodyDefinitions [{
      boundBone {boneName=dyn_collarA_r_01}
      boxExtents=<20 30 30>
      constraintSetup {angularLimitsMax=<0 5 0>, angularLimitsMin=<0 -5 0>, angularTarget=<1 0 0>}
      localJointOffset=<4 0 0>
     }]
    relativeSpaceBone {boneName=root}
   }
   Save cached pose 'preSim1Cache'(Pose=Component To Local(ComponentPose=@n33)) {cacheName=preSim1Cache}
   n34 = "Transform (Modify) Bone - Bone: dyn_vestBk_l_MSTR"!(
    ComponentPose=Local To Component(LocalPose=Use cached pose 'preSim1Cache'())
    bAlphaBoolEnabled:bool=True
    Translation="0,-4,0"
    Rotation="0,-30,0"
    Scale="1,1,1"
   ) -> (Pose) {
    alphaInputType=Bool
    rotationMode=BMM_Additive
    rotationSpace=BCS_BoneSpace
    scaleMode=BMM_Additive
    scaleSpace=BCS_BoneSpace
    translationMode=BMM_Additive
    translationSpace=BCS_BoneSpace
    alphaBoolBlend {blendInTime=0.2, blendOption=Linear, blendOutTime=0.2}
    boneToModify {boneName=dyn_vestBk_l_MSTR}
    rotation {yaw=-30}
    translation {y=-4}
   }
   n35 = "Transform (Modify) Bone - Bone: dyn_vestBk_c_MSTR"!(ComponentPose=@n34, bAlphaBoolEnabled:bool=True, Translation="-3,-2,0", Rotation="0,-30,0", Scale="0.8,1,1") -> (Pose) {
    alphaInputType=Bool
    rotationMode=BMM_Additive
    rotationSpace=BCS_BoneSpace
    scaleMode=BMM_Additive
    scaleSpace=BCS_BoneSpace
    translationMode=BMM_Additive
    translationSpace=BCS_BoneSpace
    alphaBoolBlend {blendInTime=0.2, blendOption=Linear, blendOutTime=0.2}
    boneToModify {boneName=dyn_vestBk_c_MSTR}
    rotation {yaw=-30}
    scale {x=0.8}
    translation=<-3 -2>
   }
   n36 = "Transform (Modify) Bone - Bone: dyn_vestBk_r_MSTR"!(ComponentPose=@n35, bAlphaBoolEnabled:bool=True, Translation="0,-2,0", Rotation="0,0,0", Scale="1,1,1") -> (Pose) {
    alphaInputType=Bool
    rotationMode=BMM_Additive
    rotationSpace=BCS_BoneSpace
    scaleMode=BMM_Additive
    scaleSpace=BCS_BoneSpace
    translationMode=BMM_Additive
    translationSpace=BCS_BoneSpace
    alphaBoolBlend {blendInTime=0.2, blendOption=Linear, blendOutTime=0.2}
    boneToModify {boneName=dyn_vestBk_r_MSTR}
    translation {y=-2}
   }
   Save cached pose 'PreSim2Cache'(Pose=Component To Local(ComponentPose=@n36)) {cacheName=PreSim2Cache}
  ]}
 EventGraph {nodes [
   n1 = Event BlueprintUpdateAnimation()
   > Set Delta Time X(Delta Time X=@n1.DeltaTimeX)
   > n2 = Sequence!(execute) -> (then_0, then_1, then_2:Exec, then_3:Exec, then_4:Exec, then_5:Exec, then_6:Exec)
   > Set BodyMesh(BodyMesh=GetOwnerSkeletalMeshComponent())
   n3 = Gravity Override State Function DEPRECATED(
    execute=@n2.then_1
    CharacterPartAnimInstance=Self-Reference()
    GameTimeInSeconds=Get Delta Time X()
    GravityOverrideParams=Get Gravity Override Params()
   )
   > Set Gravity Override(Gravity Override=@n3)
   Set headRotYNeg(
    execute=@n2.then_5
    headRotYNeg:float=GetAlphaFromReferenceTransformDelta(
     SkeletalMeshComponent=Get BodyMesh() -> (BodyMesh:Skeletal Mesh Component Object Reference)
     BoneName=head
     BaseName=spine_05
     DeltaComponentType=EulerY
     InputRange=(X=0,Y=40)
     TwistAxis="1,0,0"
    )
   ) -> (Output_Get:float)
   Set headRotZNeg(
    execute=@n2.then_2
    headRotZNeg:float=GetAlphaFromReferenceTransformDelta(SkeletalMeshComponent=Get BodyMesh(), BoneName=head, BaseName=spine_05, InputRange=(X=0,Y=-100), TwistAxis="1,0,0")
   ) -> (Output_Get:float)
   Set headTwistNeg(
    execute=@n2.then_3
    headTwistNeg:float=GetAlphaFromReferenceTransformDelta(
     SkeletalMeshComponent=Get BodyMesh()
     BoneName=head
     BaseName=spine_05
     DeltaComponentType=QuaternionTwist
     InputRange=(X=0,Y=-70)
     TwistAxis="1,0,0"
    )
   ) -> (Output_Get:float)
   Set headRotX(
    execute=@n2.then_6
    headRotX:float=GetAlphaFromReferenceTransformDelta(SkeletalMeshComponent=Get BodyMesh(), BoneName=head, BaseName=spine_05, DeltaComponentType=EulerX, InputRange=(X=20,Y=40), TwistAxis="1,0,0")
   ) -> (Output_Get:float)
   Set headTwistPos(
    execute=@n2.then_4
    headTwistPos:float=GetAlphaFromReferenceTransformDelta(
     SkeletalMeshComponent=Get BodyMesh()
     BoneName=head
     BaseName=spine_05
     DeltaComponentType=QuaternionTwist
     InputRange=(X=0,Y=70)
     TwistAxis="1,0,0"
    )
   ) -> (Output_Get:float)
   n4 = Event OnCharacterCustomizationLoaded()
   > Set TailMesh(TailMesh:Skeletal Mesh Component Object Reference=Get Array Item with EFort Custom Part Type(Array=@n4.CustomMeshParts, PartType=MiscOrTail)) -> (Output_Get:Skeletal Mesh Component Object Reference)
   > n5 = Sequence!(execute) -> (then_0, then_1, then_2)
   n6 = CheckForBackbling()
   > Set isWearingCape(isWearingCape=OR Boolean(A=@n6.bIsBackblingACape, B=@n6.bIsBackblingLong))
   > Set characterHasTail(characterHasTail:bool=CheckForTail!(self:Custom Character Part Anim Instance Object Reference) -> (bHasTail:bool)) -> (Output_Get:bool)
   Branch(execute=@n5.then_1, Condition=Get characterHasTail() -> (characterHasTail:bool))
   > n7 = Branch(Condition=Get isWearingCape() -> (isWearingCape:bool))
   n8 = Get TailMesh() -> (TailMesh:Skeletal Mesh Component Object Reference)
   > SetVisibility!(execute, self:Scene Component Object Reference=@n8, bNewVisibility:bool=false, bPropagateToChildren:bool=false) -> (then)
   SetVisibility(execute=@n7.else, self=@n8, bNewVisibility=true, bPropagateToChildren=false)
  ]}
]
variables [
 BodyMesh {type=SkeletalMeshComponent}
 characterHasTail {type=boolean}
 Delta Time X {type=number}
 Gravity Override {type=Vector}
 Gravity Override Params {type=GravityOverrideParamsStruct}
 headRotX {type=number}
 headRotYNeg {type=number}
 headRotZNeg {type=number}
 headTwistNeg {type=number}
 headTwistPos {type=number}
 isWearingCape {type=boolean}
 TailMesh {type=SkeletalMeshComponent}
]
